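# Translucent unlit Niagara mesh-particle material for the HomeRange skydiving contrail trail.
# Panning smoke, trail and caustic textures, distorted by noise and shaped by UV masks, drive a color-curve atlas and opacity with depth, tail and near-camera fades.
# Spline thickening with a bulge, sine wobble, back offset and Niagara-fed bend form world-position-offset branches, tuned through dynamic parameters.
# TalonPane parameters optionally blend in fire-noise emissive color and a bottom opacity fade.
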Material /BRCosmetics/Athena/Effects/Particles/Skydiving/NS_Contrail_HomeRange/M_Contrail_HomeRange_Trail
graph {
 properties {
  bUsedWithNiagaraMeshParticles=true
  blendMode=BLEND_Translucent
  pixelDepthOffsetMode=PDOM_Legacy
  shadingModel=MSM_Unlit
  usedShadingModels=Unlit
 }
 nodes [
  Add() {constB=0.25}
  n1 = TextureCoordinate()
  n2 = ComponentMask(None=@n1) {g=true}
  Add(A=Multiply(A=Add(A=@n2) {constB=-0.5}) {constB=2}) {constB=0.5}
  Constant3Vector() {constant=<rgba 0.4798 0.5661 0.6979 1>}
  Constant() {r=-0.5}
  Constant() {r=1.5}
  n3 = NamedRerouteUsage()
  n4 = ComponentMask(None=@n3) {r=true}
  n5 = Multiply(A=ComponentMask(None=@n3) {g=true}) {constB=0.5}
  n6 = Multiply(A=Power(Base=OneMinus(None=@n4)), B=Power(Base=Sine(None=@n5)) {constExponent=3})
  n7 = Saturate(None=Add(A=@n6) {constB=0.1})
  CurveAtlasRowParameter(CurveTime=@n7) {
   atlas=&/CA_HomeRange_Colors
   curve=&/CC_Contrail_HomeRange_02
   defaultValue=1
   parameterName=Color Curve
  }
  n8 = MaterialFunctionCall(X=Constant() {r=-1}, Y=Constant()) {materialFunction=&MakeFloat2}
  n9 = ScalarParameter() {defaultValue=1, parameterName=GlobalPanSpeedMult}
  n10 = TextureSample(UVs=Panner(Coordinate=NamedRerouteUsage(), Speed=Multiply(A=@n8.Result, B=@n9)) {speedX=-1}) {samplerType=SAMPLERTYPE_Grayscale, texture=&T_Smoke_Trail_Soft}
  n11 = Constant() {r=-0.1}
  n12 = Constant() {r=0.2}
  n13 = Panner(
   Coordinate=NamedRerouteUsage()
   Speed=Multiply(
    A=MaterialFunctionCall(X=@n11, Y=@n12) {materialFunction=&MakeFloat2}.Result
    B=ScalarParameter() {defaultValue=1, parameterName=GlobalPanSpeedMult}
   )
  ) {speedX=-0.1, speedY=0.2}
  n14 = Multiply(
   A=ComponentMask(None=TextureSample(UVs=@n13) {samplerType=SAMPLERTYPE_Masks, texture=/Game/Weapons/WeaponSkins/Wraps/Textures/Mash/Mash_Noise}.RGB) {g=true, r=true}
   B=ScalarParameter() {defaultValue=0.12, parameterName=DistortionMult}
  ) {constB=0.12}
  n15 = Add(A=NamedRerouteUsage(), B=@n14)
  n16 = AppendVector(A=DynamicParameter() {defaultValue=<rgba 200 50 1 0>, paramNames [WidthBase, WidthTip, U_Scale, U offset]}.U offset, B=Constant())
  n17 = MaterialFunctionCall(X=Constant() {r=-1}, Y=Constant()) {materialFunction=&MakeFloat2}
  n18 = ScalarParameter() {defaultValue=1, parameterName=GlobalPanSpeedMult}
  n19 = TextureSample(UVs=Panner(Coordinate=Add(A=@n15, B=@n16), Speed=Multiply(A=@n17.Result, B=@n18)) {speedX=-1}) {texture=/Game/Gadgets/Assets/VinderTech_GliderChute/Glider_DarkStance/FX/T_Glider_DarkStance_Trail}
  n20 = Power(Base=Add(A=Multiply(A=@n10.R), B=Multiply(A=@n19.G) {constB=2}) {constB=0})
  n21 = AppendVector(A=DynamicParameter() {defaultValue=<rgba 200 50 1 0>, paramNames [WidthBase, WidthTip, U_Scale, U offset]}.U_Scale, B=Constant() {r=1})
  n22 = NamedRerouteUsage()
  n23 = Add(
   A=Add(A=Multiply(A=@n21, B=@n22), B=@n14)
   B=AppendVector(A=DynamicParameter() {defaultValue=<rgba 200 50 1 0>, paramNames [WidthBase, WidthTip, U_Scale, U offset]}.U offset, B=Constant())
  )
  n24 = Multiply(
   A=MaterialFunctionCall(X=Constant() {r=-1}, Y=Constant()) {materialFunction=&MakeFloat2}.Result
   B=ScalarParameter() {defaultValue=1, parameterName=GlobalPanSpeedMult}
  )
  n25 = MaterialFunctionCall(X=Constant() {r=-0.5}, Y=Constant()) {materialFunction=&MakeFloat2}
  n26 = ScalarParameter() {defaultValue=1, parameterName=GlobalPanSpeedMult}
  n27 = TextureSample(UVs=Panner(Coordinate=NamedRerouteUsage(), Speed=Multiply(A=@n25.Result, B=@n26)) {speedX=-0.5}) {
   samplerType=SAMPLERTYPE_LinearColor
   texture=/Engine/Functions/Engine_MaterialFunctions02/ExampleContent/Textures/OffsetNoiseDistanceFields_ForNormals
  }
  n28 = Multiply(
   A=TextureSample(UVs=Panner(Coordinate=@n23, Speed=@n24) {speedX=-1}) {texture=&T_Glider_DarkStance_Trail}.R
   B=LinearInterpolate(Alpha=@n27.R) {constA=0.4}
  )
  n29 = Saturate(None=Add(A=Add(A=@n20, B=@n28)) {constB=0})
  n30 = LinearInterpolate(Alpha=@n27.R) {constA=0.5}
  n31 = NamedRerouteUsage()
  n32 = LinearInterpolate(Alpha=Power(Base=ComponentMask(None=@n31) {r=true}) {constExponent=1.5})
  n33 = NamedRerouteUsage()
  n34 = Multiply(
   A=Multiply(A=@n29, B=Power(Base=@n30, Exp=@n32) {constExponent=1})
   B=LinearInterpolate(Alpha=ComponentMask(None=@n33) {r=true}) {constA=0.8}
  )
  n35 = DynamicParameter() {defaultValue=<rgba 200 50 1 0>, paramNames [WidthBase, WidthTip, U_Scale, U offset]}
  n36 = Constant() {r=1}
  n37 = Constant() {r=-1}
  n38 = Constant()
  n39 = Panner(
   Coordinate=Multiply(A=AppendVector(A=@n35.U_Scale, B=@n36), B=NamedRerouteUsage())
   Speed=Multiply(
    A=MaterialFunctionCall(X=@n37, Y=@n38) {materialFunction=&MakeFloat2}.Result
    B=ScalarParameter() {defaultValue=1, parameterName=GlobalPanSpeedMult}
   )
  ) {speedX=-1}
  n40 = Multiply(A=@n14)
  n41 = Power(Base=TextureSample(UVs=Add(A=@n39, B=@n40)) {samplerType=SAMPLERTYPE_LinearGrayscale, texture=/Game/Global/Textures/Default/Caustics/T_Nosie_Caustics_Grayscale}.R) {constExponent=1}
  n42 = OneMinus(None=ComponentMask(None=NamedRerouteUsage()) {r=true})
  n43 = Multiply(A=Multiply(A=Multiply(A=@n41) {constB=1.5}, B=Power(Base=@n42) {constExponent=1}), B=@n29)
  n44 = Multiply(A=@n7, B=Saturate(None=Add(A=@n34, B=@n43)))
  n45 = DynamicParameter() {parameterIndex=1, defaultValue=<rgba 1 1 0 0>, paramNames [Use Forward Tail mask, Appear Mask, BackOffset, Use Trailing Fade]}
  n46 = Multiply(A=@n14) {constB=0.5}
  n47 = Multiply(A=NamedRerouteUsage()) {constB=0.3}
  n48 = DynamicParameter() {defaultValue=<rgba 200 50 1 0>, paramNames [WidthBase, WidthTip, U_Scale, U offset]}
  n49 = Constant()
  n50 = Constant() {r=-0.5}
  n51 = Constant()
  n52 = Panner(
   Coordinate=Add(A=Add(A=@n46, B=@n47), B=AppendVector(A=@n48.U offset, B=@n49))
   Speed=Multiply(
    A=MaterialFunctionCall(X=@n50, Y=@n51) {materialFunction=&MakeFloat2}.Result
    B=ScalarParameter() {defaultValue=1, parameterName=GlobalPanSpeedMult}
   )
  ) {speedX=-0.5}
  n53 = Add(A=TextureSample(UVs=@n52) {samplerType=SAMPLERTYPE_Masks, texture=&Mash_Noise}.R) {constB=0.001}
  n54 = DynamicParameter() {parameterIndex=1, defaultValue=<rgba 1 1 0 0>, paramNames [Use Forward Tail mask, Appear Mask, BackOffset, Use Trailing Fade]}
  n55 = ComponentMask(None=NamedRerouteUsage()) {r=true}
  n56 = Saturate(None=LinearInterpolate(
    A=Constant() {r=-1}
    B=LinearInterpolate(Alpha=@n54.Use Forward Tail mask) {constA=0.5, constB=1.5}
    Alpha=Power(Base=@n55) {constExponent=1.5}
   ) {constA=0.1, constB=2})
  n57 = Saturate(None=LinearInterpolate(A=LinearInterpolate(Alpha=@n45.Use Forward Tail mask) {constB=-0.2}, B=Constant() {r=1.2}, Alpha=Power(Base=@n53, Exp=@n56)) {constA=-2, constB=2})
  n58 = NamedRerouteUsage()
  n59 = OneMinus(None=ComponentMask(None=@n58) {r=true})
  n60 = Multiply(A=@n59, B=Constant() {r=10}) {constB=10}
  n61 = Constant()
  n62 = SmoothStep(Min=@n61, Max=Add(A=@n61, B=Constant() {r=0.1}), Value=@n59)
  n63 = DynamicParameter() {parameterIndex=1, defaultValue=<rgba 1 1 0 0>, paramNames [Use Forward Tail mask, Appear Mask, BackOffset, Use Trailing Fade]}
  n64 = Multiply(A=Multiply(A=@n44, B=@n57), B=LinearInterpolate(A=Saturate(None=@n60), B=Saturate(None=@n62), Alpha=@n63.Use Trailing Fade) {constA=1})
  n65 = NamedRerouteUsage()
  n66 = NamedRerouteUsage()
  n67 = Multiply(A=Power(Base=ComponentMask(None=@n65) {r=true}) {constExponent=1.2}, B=Saturate(None=Multiply(A=@n66)))
  n68 = LinearInterpolate(Alpha=Saturate(None=Multiply(A=@n67) {constB=50})) {constA=0.85}
  n69 = NamedRerouteUsage()
  n70 = ScalarParameter() {parameterName=UseSmokeLines}
  n71 = CurveAtlasRowParameter(CurveTime=Add(A=Multiply(A=@n64, B=@n68), B=LinearInterpolate(B=@n69, Alpha=@n70))) {
   atlas=&CA_HomeRange_Colors
   curve=&CC_Contrail_HomeRange_02
   defaultValue=1
   parameterName=Color Curve
  }
  n72 = ParticleColor()
  n73 = Multiply(A=@n72.A, B=@n64)
  n74 = OneMinus(None=ComponentMask(None=NamedRerouteUsage()) {r=true})
  n75 = DynamicParameter() {parameterIndex=1, defaultValue=<rgba 1 1 0 0>, paramNames [Use Forward Tail mask, Appear Mask, BackOffset, Use Trailing Fade]}
  n76 = LinearInterpolate(Alpha=@n75.Appear Mask) {constA=-1, constB=0.1}
  n77 = Saturate(None=Multiply(A=Add(A=@n74, B=@n76)) {constB=10})
  n78 = Multiply(A=Multiply(A=@n71.A, B=Multiply(A=@n73, B=@n77)))
  n79 = DepthFade(Opacity=@n78, FadeDistance=ScalarParameter() {parameterName=DepthFade})
  n80 = Multiply(A=Multiply(A=ShadingPathSwitch(Default=@n79, Deferred=@n79, Forward=@n78, Mobile=@n78), B=NamedRerouteUsage()), B=NamedRerouteUsage())
  n81 = TextureCoordinate()
  n82 = VectorParameter() {group=TalonPane, parameterName=BottomFadeControl, channelNames {a="", b=Multiply, g=Power, r=Impact}, defaultValue=<rgba 1 1.5 10 0>}
  n83 = Power(Base=OneMinus(None=ComponentMask(None=@n81) {r=true}), Exp=@n82.Power) {constExponent=1.5}
  n84 = Saturate(None=Subtract(A=Multiply(A=@n83, B=@n82.Multiply) {constB=10}, B=NamedRerouteUsage()))
  LinearInterpolate(A=@n80, B=Multiply(A=Multiply(A=@n80, B=@n84)), Alpha=ScalarParameter() {group=TalonPane, parameterName=UseTalonPaneOpacity, sortPriority=4})
  n85 = Multiply(A=TextureCoordinate(), B=Constant2Vector() {g=3, r=0.2})
  n86 = MaterialFunctionCall(Multiplier=Constant() {r=2}) {materialFunction=&Pi}
  n87 = TextureSample(UVs=Panner(Coordinate=Rotator(Coordinate=@n85, Time=@n86.Result)) {speedX=0.2, speedY=0.1}) {texture=/BRCosmetics/Gadgets/Assets/VinderTech_GliderChute/Glider_BlondeJaw/FX/T_Cloud_Noise}
  n88 = TextureCoordinate()
  n89 = Divide(A=ConstantBiasScale(None=@n88) {bias=-0.5, scale=2}, B=Constant2Vector() {g=0.75, r=2}) {constB=1}
  n90 = DotProduct(A=@n89, B=@n89)
  n91 = SquareRoot(None=OneMinus(None=Saturate(None=@n90)))
  n92 = ConstantBiasScale(None=Divide(A=@n89, B=Add(A=@n91)))
  n93 = Multiply(A=@n87.R) {constB=0.01}
  n94 = Multiply(A=Panner(Coordinate=Add(A=@n92, B=@n93)) {speedX=0.5}, B=Constant2Vector() {g=1, r=8})
  LinearInterpolate(A=@n87.R, B=@n87.R, Alpha=SmoothStep(Value=Frac(None=@n94)))
  n95 = Multiply(
   A=LinearInterpolate(A=@n72.RGB, B=NamedRerouteUsage(), Alpha=ScalarParameter() {group=TalonPane, parameterName=UseTalonPaneEmissive, sortPriority=0})
   B=@n71
  )
  n96 = Desaturation(None=@n95)
  n97 = ParticleColor()
  MaterialFunctionCall(Emissive=LinearInterpolate(A=@n95, B=Multiply(A=@n96, B=@n97.RGB), Alpha=ScalarParameter() {parameterName=Use Desaturation})) {materialFunction=&MF_FX_EmissiveAdjustments}
  n98 = NamedRerouteUsage()
  n99 = Add(A=Add(A=ComponentMask(None=@n1) {r=true}, B=Multiply(A=@n98) {constB=0.2})) {constB=-0.5}
  n100 = Multiply(A=Sine(None=Multiply(A=@n2) {constB=0.5})) {constB=0.8}
  n101 = LinearInterpolate(Alpha=Power(Base=Saturate(None=@n100))) {constA=2, constB=0.2}
  MaterialFunctionCall(R=Add(A=Multiply(A=@n99, B=@n101)) {constB=0.5}) {materialFunction=&AppendMany}
  n102 = TextureCoordinate()
  MaterialFunctionCall(XY=@n102) {materialFunction=&Swizzle}
  n103 = TextureCoordinate()
  Multiply(A=Power(Base=ComponentMask(None=@n103) {r=true}, Exp=ScalarParameter() {defaultValue=1, parameterName=Gradient Power}), B=@n73)
  Multiply(A=@n6) {constB=5}
  Multiply(A=@n87.R) {constB=-1}
  n104 = DynamicParameter() {defaultValue=<rgba 200 50 1 0>, paramNames [WidthBase, WidthTip, U_Scale, U offset]}
  n105 = TextureCoordinate()
  n106 = Multiply(A=Power(Base=ComponentMask(None=@n105) {r=true}) {constExponent=20}) {constB=2}
  n107 = LinearInterpolate(A=ScalarParameter() {group=SplineThicken, parameterName=BottomThickness}, Alpha=OneMinus(None=Saturate(None=@n106)))
  n108 = Multiply(A=Add(A=@n104.WidthBase, B=ScalarParameter() {defaultValue=5, group=SplineThicken, parameterName=Base}), B=@n107)
  n109 = DynamicParameter() {parameterIndex=2, defaultValue=<rgba 1 100 -1 0>, paramNames [Speed Dir Mult, NearCameraFade, BulgePosition, Param4]}
  n110 = Add(
   A=ComponentMask(None=TextureCoordinate()) {r=true}
   B=Add(A=ScalarParameter() {group=SplineThicken, parameterName=BulgePosition}, B=@n109.BulgePosition)
  )
  n111 = Saturate(None=Sine(None=Multiply(A=@n110) {constB=0.5}))
  n112 = ScalarParameter() {defaultValue=5, group=SplineThicken, parameterName=BulgeGradientPower}
  n113 = Multiply(
   A=@n107
   B=Multiply(A=Power(Base=@n111, Exp=@n112) {constExponent=5}, B=ScalarParameter() {defaultValue=100, group=SplineThicken, parameterName=BulgeAmount}) {constB=50}
  )
  n114 = Multiply(A=@n104.WidthTip, B=@n107)
  NamedRerouteDeclaration(None=MaterialFunctionCall(WidthBase=Add(A=@n108, B=@n113), WidthTip=Add(A=@n114, B=@n113), Expand U or V UV Channel=StaticBool()) {materialFunction=&SplineThicken}.WorldPositionOffset) {name=SplineThicken, nodeColor=<rgba 0 1 1 1>}
  n115 = MaterialFunctionCall(Float2=@n102) {materialFunction=&BreakOutFloat2Components}
  NamedRerouteDeclaration(None=MaterialFunctionCall(X=OneMinus(None=@n115.R), Y=@n115.G) {materialFunction=&MakeFloat2}.Result) {name=UV's, nodeColor=<rgba 0.7812 1 0 1>}
  n116 = TextureSample(UVs=Multiply(A=NamedRerouteUsage()) {constB=1.5}) {texture=/Game/Effects/Fort_Effects/Textures/Fire/HF_FireNoisesPacked}
  NamedRerouteDeclaration(None=@n116.B) {name=FlameNoise, nodeColor=<rgba 0 0.0625 1 1>}
  n117 = ComponentMask(None=TextureCoordinate()) {r=true}
  n118 = Power(Base=TextureSample(UVs=@n94) {texture=/BRCosmetics/Characters/Player/Female/Medium/Bodies/F_MED_CowboyHat/Skins/Steer/Textures/BlackWhite_Gradient}.R)
  NamedRerouteDeclaration(None=Saturate(None=Multiply(A=@n117, B=@n118))) {name=Smoke Lines, nodeColor=<rgba 0 0.2969 1 1>}
  n119 = NamedRerouteUsage()
  n120 = Power(Base=ComponentMask(None=NamedRerouteUsage()) {r=true}, Exp=Constant() {r=1.5})
  n121 = DynamicParameter() {parameterIndex=1, defaultValue=<rgba 1 1 0 0>, paramNames [Use Forward Tail mask, Appear Mask, BackOffset, Use Trailing Fade]}
  n122 = Transform(None=MaterialFunctionCall(X=Constant(), Y=@n121.BackOffset, Z=ScalarParameter() {defaultValue=-50, parameterName=Back Offset}) {materialFunction=&MakeFloat3}.Result) {transformSourceType=TRANSFORMSOURCE_World, transformType=TRANSFORM_Local}
  n123 = Add(A=Add(A=@n119, B=Multiply(A=@n120, B=@n122)), B=NamedRerouteUsage()) {constB=0}
  n124 = NamedRerouteUsage()
  NamedRerouteDeclaration(None=LinearInterpolate(A=Add(A=@n123, B=@n124), B=@n119, Alpha=ScalarParameter() {parameterName=WPO Distort OFF})) {name=WPO, nodeColor=<rgba 0.17 0.0489 0 1>}
  n125 = NamedRerouteUsage()
  n126 = DynamicParameter() {defaultValue=<rgba 200 50 1 0>, paramNames [WidthBase, WidthTip, U_Scale, U offset]}
  n127 = AppendVector(A=@n126.U offset, B=Constant())
  n128 = Constant() {r=1}
  n129 = ComponentMask(None=Multiply(A=Add(A=@n125, B=@n127), B=AppendVector(A=@n126.U_Scale, B=@n128))) {r=true}
  n130 = Multiply(A=Time()) {constB=-0.7}
  n131 = Multiply(
   A=Sine(None=Add(A=@n129, B=@n130))
   B=Transform(None=VectorParameter() {parameterName=WPO Direction, defaultValue=<rgba 0 1 0 1>}.RGB) {transformSourceType=TRANSFORMSOURCE_Camera}
  )
  n132 = ScalarParameter() {defaultValue=5, parameterName=WPO Intensity}
  n133 = Multiply(A=ComponentMask(None=NamedRerouteUsage()) {r=true}) {constB=7}
  NamedRerouteDeclaration(None=Multiply(A=Multiply(A=@n131, B=@n132), B=Saturate(None=@n133))) {name=WPO Distortion, nodeColor=<rgba 0.1953 1 0 1>}
  n134 = DynamicParameter() {desc=Parameters set by Niagara, parameterIndex=3, paramNames [X,Y,Z,WPOIntensity]}
  n135 = MaterialFunctionCall(X=@n134.X, Y=@n134.X, Z=@n134.X) {materialFunction=&MakeFloat3}
  n136 = ParticlePositionWS()
  n137 = Clamp(
   None=Multiply(B=Subtract(A=@n135.Result, B=@n136)) {constA=1}
   Min=ScalarParameter() {defaultValue=-35, group=WPO, parameterName=Clamp_min_X}
   Max=ScalarParameter() {defaultValue=35, group=WPO, parameterName=Clamp_max}
  ) {maxDefault=1000, minDefault=-1000}
  n138 = NamedRerouteUsage()
  n139 = Multiply(
   A=Power(Base=ComponentMask(None=@n138) {r=true}, Exp=ScalarParameter() {defaultValue=3, group=WPO, parameterName=WPOBendPower})
   B=@n134.WPOIntensity
  )
  NamedRerouteDeclaration(None=LinearInterpolate(B=LinearInterpolate(B=@n137, Alpha=@n139), Alpha=ScalarParameter() {defaultValue=1, parameterName=Use interpolate})) {name=WPO Interpolate, nodeColor=<rgba 1 0 0.0703 1>}
  n140 = DynamicParameter() {parameterIndex=2, defaultValue=<rgba 1 100 -1 0>, paramNames [Speed Dir Mult, NearCameraFade, BulgePosition, Param4]}
  n141 = Power(
   Base=MaterialFunctionCall(Distance=@n140.NearCameraFade) {materialFunction=&Near_Camera_Fade}.Result
   Exp=ScalarParameter() {defaultValue=2, parameterName=NearCamFadePower}
  )
  n142 = Constant() {r=1}
  NamedRerouteDeclaration(None=QualitySwitch(Default=@n141, Low=@n142, High=@n141, Medium=@n142, Epic=@n141)) {name=NearCameraFade, nodeColor=<rgba 1 0 0.7734 1>}
  n143 = VectorParameter() {group=TalonPane, parameterName=TalonPane_EmissiveColor, sortPriority=1, defaultValue=<rgba 0.3763 0.0802 0.0232 1>}
  n144 = TextureCoordinate()
  n145 = ComponentMask(None=@n144) {r=true}
  n146 = LinearInterpolate(A=ParticleColor().RGB, B=@n143.RGB, Alpha=Power(Base=@n145) {constExponent=8})
  n147 = VectorParameter() {group=TalonPane, parameterName=TalonPane_FireNoiseColorLow, sortPriority=3, defaultValue=<rgba 5 0.1986 0.033 1>}
  n148 = VectorParameter() {group=TalonPane, parameterName=TalonPane_FireNoiseColorHigh, sortPriority=2, defaultValue=<rgba 20 6.5856 0.132 1>}
  n149 = Power(Base=@n145) {constExponent=15}
  n150 = Multiply(
   A=Power(Base=@n145) {constExponent=10}
   B=LinearInterpolate(B=LinearInterpolate(A=@n147.RGB, B=@n148.RGB, Alpha=@n149), Alpha=Multiply(A=@n116.B) {constB=4})
  )
  n151 = Multiply(A=ComponentMask(None=@n144) {g=true}) {constB=0.5}
  n152 = Multiply(A=Power(Base=Sine(None=@n151)) {constExponent=4})
  NamedRerouteDeclaration(None=Multiply(A=Add(A=@n146, B=@n150), B=LinearInterpolate(Alpha=@n152) {constA=0.05})) {name=TalonPaneEmissive, nodeColor=<rgba 0 0.0625 1 1>}
  n153 = Constant() {r=0.436}
  n154 = Add(A=@n153, B=Constant() {r=0.2})
  n155 = MaterialFunctionCall(Float2=TextureCoordinate()) {materialFunction=&BreakOutFloat2Components}
  n156 = OneMinus(None=Abs(None=ConstantBiasScale(None=@n155.G) {bias=-0.5, scale=2}))
  n157 = Constant() {r=0.8}
  n158 = Add(A=@n157, B=Constant() {r=0.2})
  n159 = MaterialFunctionCall(Float2=TextureCoordinate()) {materialFunction=&BreakOutFloat2Components}
  NamedRerouteDeclaration(None=LinearInterpolate(A=Constant() {r=1}, B=SmoothStep(Min=@n153, Max=@n154, Value=@n156), Alpha=SmoothStep(Min=@n157, Max=@n158, Value=@n159.R)) {constA=1}) {name=Bottom Mask, nodeColor=<rgba 1 0.2578 0 1>}
  NamedRerouteDeclaration(None=@n52) {name=FlamesPanner, nodeColor=<rgba 1 0 0.1875 1>}
  NamedRerouteDeclaration(None=@n43) {name=Distorted Noise, nodeColor=<rgba 0.5938 0 1 1>}
  NamedRerouteUsage()
  NamedRerouteUsage()
  OneMinus(None=@n138)
  n160 = ComponentMask(None=Add(A=@n88) {constB=0}) {r=true}
  Panner(Coordinate=Sine(None=Multiply(A=@n160) {constB=2})) {speedX=0.2}
  n161 = Constant()
  n162 = SmoothStep(Min=@n161, Max=Add(A=@n161, B=Constant() {r=0.025}), Value=ComponentMask(None=NamedRerouteUsage()) {r=true})
  n163 = ScalarParameter() {defaultValue=1, parameterName=LeadEdge Opacity}
  Reroute(None=Multiply(A=Multiply(A=@n44, B=@n57), B=LinearInterpolate(B=@n162, Alpha=@n163) {constA=1}))
  Sine(None=Multiply(A=ComponentMask(None=@n58) {g=true}) {constB=0.5})
  Subtract(A=@n118, B=OneMinus(None=@n117))
  TextureCoordinate()
  TextureCoordinate()
 ]
}
