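# Niagara skydiving contrail cosmetic with red-tinted fire trails, embers, feathers and pulsing glow, opened by a brief anticipation flash.
# Anticipation, Embers, Fire and FireTrail_And_Dots follow particles of a Soul emitter through attribute readers, while Feathers and Glow sample skeletal mesh bones or sockets.
# User.ParaGlideLeanAlpha drives spawning, motion, scale and color.
# Front-end modifiers swap gameplay and menu values, and spawn-group switches let single emitters render several particle variants.
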NiagaraSystem /BRCosmetics/Athena/Effects/Particles/Skydiving/NS_Contrail_HomeRange/NS_HomeRange_Contrail
dependencies {
 usedDataInterfaces [
  &NiagaraDataInterfaceCamera
  &NiagaraDataInterfaceCurve
  &NiagaraDataInterfaceParticleRead
  &NiagaraDataInterfaceVectorCurve
  &NiagaraDataInterfaceColorCurve
  &NiagaraDataInterfaceSpriteRendererInfo
  &NiagaraDataInterfaceVector2DCurve
  &NiagaraDataInterfaceSkeletalMesh
 ]
 usedDynamicInputs [
  &VectorFromCurve
  &Vector2DFromCurve
  &ConvertVectorToPosition
 ]
 usedModules [
  &self:SystemSpawnScript.SystemScriptSource.SystemScriptGraph.NiagaraNodeAssignment_2.SetVariables_#1
  &SystemState
  /Game/Effects/Niagara/Modules/Skydiving/FN_DivingAttributes/v5/FN_DivingAttributes
  &self:SystemSpawnScript.SystemScriptSource.SystemScriptGraph.NiagaraNodeAssignment_4.SetVariables_#2
  &FN_Integrator
 ]
}
emitters {
 Anticipation {
  emitterUpdateScript [
   &EmitterState(Life Cycle Mode=::NewEnumerator1, Loop Behavior=::NewEnumerator1, Loop Delay=0.3, Loop Duration=1, UseLoopDelay=true)
   &SpawnBurst_Instantaneous(Spawn Count=2, Spawn Group=1, Spawn Time=0.5, Use Spawn Probability=false, UseLoopCountLimit=false)
   &SpawnRate(
    SpawnRate=&FloatFromCurve(
     FloatCurve=DataInterface() {propertyValues {Curve {keys [<at 0.1 1>,<at 0.15 0>]}}}
     CurveIndex=&Subtract_Float(A=$Emitter.Age, B=0.4)
     Scale Curve=100
    )
    Use Spawn Probability=false
   )
  ]
  particleSpawnScript [
   /SetVariables(Particles.VisibilityTag=&SpawnGroup()) {isSetParameters=true}
   &InitializeParticle(
    Color=&MergeHelper_ColorSwitch(0=<rgba 3 0.3285 0.2998 1>, 1=<rgba 1 0 0.0605 5>, 2=?, 3=?, Index=$Particles.VisibilityTag, Num=::NewEnumerator0)
    Lifetime Max=0.15
    Lifetime Min=0.12
    Lifetime Mode=::NewEnumerator1
    Mass Mode=::NewEnumerator3
    Sprite Rotation Mode=::NewEnumerator2
    Sprite Size Mode=::NewEnumerator3
    Uniform Sprite Size=&MergeHelper_FloatSwitch(
     0=&RandomRangeFloat(Minimum=10, Maximum=40)
     1=100
     2=?
     3=?
     Index=$Particles.VisibilityTag
     Num=::NewEnumerator0
    )
    UsePositionOffset=false
    Write Material Random=true
   )
   &FN_SetSpawnGroup(Number of Spawn Groups=2)
   &ShapeLocation(
    MaskSpawnGroup=false
    Non Uniform Scale=<1 1 1.8>
    Override Seed=false
    Sphere Radius=&MergeHelper_FloatSwitch(0=10, 1=0, 2=?, 3=?, Index=$Particles.VisibilityTag, Num=::NewEnumerator0)
    UseOwnerScale=true
   )
   &CameraOffset(Camera Offset Amount=20)
  ]
  particleUpdateScript [
   &ParticleState()
   &ScaleSpriteSize(
    Scale Sprite Size Mode=::NewEnumerator0
    Uniform Scale Factor=&MergeHelper_FloatSwitch(
     0=&FloatFromCurve(FloatCurve=DataInterface() {propertyValues {Curve {keys [<at 0 0 RCIM_Cubic>, <at 0.2 1 RCIM_Cubic>, {interpMode=RCIM_Cubic, leaveTangent=0.5, time=1, value=0}]}}})
     1=&FloatFromCurve(FloatCurve=DataInterface() {propertyValues {Curve {keys [<at 0.2 1>, {tangentMode=RCTM_User, time=1, value=0}]}}})
     2=?
     3=?
     Index=$Particles.VisibilityTag
     Num=::NewEnumerator0
    )
   )
   &ScaleColor(
    Scale Alpha=&FloatFromCurve(FloatCurve=DataInterface() {propertyValues {Curve {keys [<at -0.0046 1.0113 RCIM_Cubic>,<at 1.0011 1 RCIM_Cubic>]}}})
    ScaleA=true
    ScaleRGB=true
   )
   &SolveForcesAndVelocity(Clamp Velocity=false, Limit Acceleration=false)
   /SetVariables(
    Particles.GlowColor=&MakeLinearColorFromVectorAndFloat(
     "Vector (RGB)"=&MakeVectorFromLinearColorRGB(Color=<rgba 1 0 0.2229 1>)
     "Float (Alpha)"=&FloatFromCurve(FloatCurve=DataInterface() {propertyValues {Curve {keys [<at 0 0>,<at 0.2 1>,<at 1 0>]}}}, Scale Curve=0.3)
    )
    Particles.GlowSize=&Multiply_Vector2D_ByFloat(Float=3.5, Vector2D=$Particles.SpriteSize)
   ) {isSetParameters=true}
   &self:SamplePosFromOtherParticle(
    Index:NiagaraInt32=&ModuloInt(A=&SpawnGroup(), B=2)
    PAR=DataInterface() {propertyValues {EmitterBinding {bindingMode=Other, emitterName=Soul}}}
   ) {module=SamplePosFromOtherParticle001}
   /SetVariables(Particles.Position=&AddVectorToPosition(
     Position=&TransformPosition(Position=$Output.SamplePosFromOtherParticle001.ParticlePos, Source Space=::World, Destination Space=::Local)
     Vector=Linked() {linkedVariable {name=Particles.Initial.ShapeLocation.ShapePosition, type=NiagaraPosition}}
    )) {isSetParameters=true}
  ]
  rendererClasses [&NiagaraSpriteRendererProperties]
  renderers [
   {rendererClass=&NiagaraSpriteRendererProperties}
   {rendererClass=&NiagaraSpriteRendererProperties, rendererIndex=1}
   {rendererClass=&NiagaraSpriteRendererProperties, rendererIndex=2}
  ]
 }
 Embers {
  emitterUpdateScript [
   &EmitterState(Delay First Loop Only=true, Life Cycle Mode=::NewEnumerator1, Loop Delay=0.3, Loop Duration=1, UseLoopDelay=true)
   &SpawnRate(
    SpawnRate=&FloatFromCurve(
     FloatCurve=DataInterface() {propertyValues {Curve {keys [<at 0.25 0>,<at 0.3 1>,<at 0.4 0>]}}}
     CurveIndex=$Engine.System.Age
     Scale Curve=200
    )
    Use Spawn Probability=false
   )
   &SpawnBurst_Instantaneous(Spawn Count=30, Spawn Group=1, Spawn Time=0.55, Use Spawn Probability=false, UseLoopCountLimit=true)
   &SpawnRate(
    Spawn Probability=&FloatFromCurve(FloatCurve=DataInterface() {propertyValues {Curve {keys [<at 0.7 0>,<at 1 1>]}}}, CurveIndex=$User.ParaGlideLeanAlpha)
    SpawnGroup=2
    SpawnRate=&FloatFromCurve(FloatCurve=DataInterface() {propertyValues {Curve {keys [<at 0.9 0>,<at 1 1>]}}}, CurveIndex=$Engine.System.Age, Scale Curve=100)
    Use Spawn Probability=true
   ) {module=SpawnRate001}
  ]
  particleSpawnScript [
   /SetVariables(Particles.PerParticleRandom:NiagaraFloat=&RandomRangeFloat(Minimum=0, Maximum=1)) {isSetParameters=true}
   /SetVariables(Particles.CustomIndex:NiagaraInt32=$Emitter.SpawnGroup) {isSetParameters=true}
   &InitializeParticle(
    Lifetime=&MergeHelper_FloatSwitch(
     0=&FloatFromCurve(
      FloatCurve=DataInterface() {propertyValues {Curve {keys [<at 0 0.2>,<at 0.9 0.3>,<at 1 0.4>]}}}
      CurveIndex=&RandomRangeFloat(Minimum=0, Maximum=1)
      Scale Curve=0.6
     )
     1=&FloatFromCurve(
      FloatCurve=DataInterface() {propertyValues {Curve {keys [<at 0 0.3>,<at 0.9 0.4>,<at 1 0.7>]}}}
      CurveIndex=&RandomRangeFloat(Minimum=0, Maximum=1)
     )
     2=&FloatFromCurve(
      FloatCurve=DataInterface() {propertyValues {Curve {keys [<at 0 0.3>,<at 0.9 0.4>,<at 1 0.7>]}}}
      CurveIndex=&RandomRangeFloat(Minimum=0, Maximum=1)
      Scale Curve=1.5
     )
     3=?
     Index=$Particles.CustomIndex
     Num=::NewEnumerator1
    )
    Mesh Scale=<2 1 1>
    Mesh Scale Mode=::NewEnumerator4
    Sprite Size=&Multiply_Vector2D(
     A=<0.4 0.8>
     B=&MakeVector2D(
      X=&FloatFromCurve(
       FloatCurve=DataInterface() {propertyValues {Curve {keys [<at 0 1>,<at 0.8 2 RCIM_Constant>,<at 0.95 3 RCIM_Constant>]}}}
       CurveIndex=$Particles.PerParticleRandom
      )
      Y=&FloatFromCurve(
       FloatCurve=DataInterface() {propertyValues {Curve {keys [<at 0 1>,<at 0.8 2 RCIM_Constant>,<at 0.95 3>]}}}
       CurveIndex=$Particles.PerParticleRandom
      )
     )
    )
    Sprite Size Mode=::NewEnumerator4
    UsePositionOffset=false
    Write Material Random=true
   )
   &FN_SetSpawnGroup(Number of Spawn Groups=2)
   &self:SamplePositionFromOtherParticle_01(PAR=DataInterface() {propertyValues {EmitterBinding {bindingMode=Other, emitterName=Soul}}}) {module=SamplePositionFromOtherParticle_01}
   /SetVariables(Particles.Position=&TransformPosition(Position=$Particles.Position, Source Space=::World, Destination Space=::Local)) {isSetParameters=true}
   &ShapeLocation(MaskSpawnGroup=false, Non Uniform Scale=<1 1 2>, Override Seed=false, Sphere Radius=7, Sphere Surface Distribution=1, UseOwnerScale=true)
   &AddVelocity(
    Velocity=&Add_Vector(
     A=&Lerp_Vector(
      A=$Particles.ShapeLocation.ShapeNormal
      B=&Multiply_VectorByFloat(Vector=$Particles.ShapeLocation.ShapeNormal, Float=0.2)
      Alpha=$User.ParaGlideLeanAlpha
     )
     B=&Lerp_Vector(
      A=<0 0 0>
      B=&FN_FrontEndVector3Modifier(Gameplay=<0 0 1.5>, Front End=<0 -0.7 1.5>)
      Alpha=$User.ParaGlideLeanAlpha
     )
    )
    Velocity Speed Scale=&RandomRangeFloat(Minimum=100, Maximum=500)
   ) {module=AddVelocity001}
  ]
  particleUpdateScript [
   &ParticleState()
   &Drag(
    Drag=&FloatFromCurve(
     FloatCurve=DataInterface() {propertyValues {Curve {keys [
         {interpMode=RCIM_Cubic, tangentMode=RCTM_User, time=0, value=0.8}
         {arriveTangent=0.4778, interpMode=RCIM_Cubic, leaveTangent=0.4778, tangentMode=RCTM_User, time=1, value=1.0375}
        ]}}}
     Scale Curve=&MergeHelper_FloatSwitch(0=15, 1=15, 2=5, 3=0, 4=0, Index=$Particles.CustomIndex, Num=::NewEnumerator3)
    )
    Use Linear Drag=true
    Use Rotational Drag=false
   )
   &CurlNoiseForce(
    Noise Strength=&MergeHelper_FloatSwitch(0=2000, 1=2000, 2=1000, 3=?, Index=$Particles.CustomIndex, Num=::NewEnumerator1)
    Pan Noise=true
    Pan Noise Field=<0 0 1>
   ) {module=CurlNoiseForce001}
   &SolveForcesAndVelocity(Clamp Velocity=false, Limit Acceleration=false)
   &ScaleColor(
    Scale Alpha=&FloatFromCurve(FloatCurve=DataInterface() {propertyValues {Curve {keys [{interpMode=RCIM_Cubic, tangentMode=RCTM_User, time=0, value=1}, {interpMode=RCIM_Cubic, tangentMode=RCTM_User, time=1, value=0}]}}})
    Scale RGB=&MakeVectorFromLinearColorRGB(Color=&ColorFromCurve(ColorCurve=DataInterface() {propertyValues {
        AlphaCurve {keys []}
        BlueCurve {keys [<at 0 2.1548 RCIM_Cubic>]}
        GreenCurve {keys [<at 0 2.0783 RCIM_Cubic>]}
        RedCurve {keys [<at 0 4 RCIM_Cubic>]}
       }}))
    ScaleA=true
    ScaleRGB=true
   )
   &ScaleSpriteSize(
    Uniform Curve Scale=&MakeFloatFromVector(Vector=$Engine.Owner.Scale, Channel=::NewEnumerator0)
    Uniform Curve Sprite Scale=DataInterface() {propertyValues {Curve {keys [
        {tangentMode=RCTM_User, time=0, value=5}
        <at 0.1 1>
        {arriveTangent=-0.0405, interpMode=RCIM_Cubic, leaveTangent=-0.0405, tangentMode=RCTM_User, time=0.7009, value=1}
        {arriveTangent=-8.0089, interpMode=RCIM_Cubic, leaveTangent=-8.0089, tangentMode=RCTM_User, time=1, value=0}
       ]}}}
   )
   &ScaleSpriteSizeBySpeed(
    Max Scale Factor=<0.5 4>
    Sample Scale Factor By Curve=true
    Scale Factor Curve=DataInterface() {propertyValues {Curve {keys [<at 0 0>, {arriveTangent=1.4989, interpMode=RCIM_Cubic, tangentMode=RCTM_Break, time=1, value=1}]}}}
    Velocity Threshold=250
   )
   /SetVariables(Particles.Glow03_Size=&Multiply_Vector2D_ByFloat(
     Float=&FloatFromCurve(
      FloatCurve=DataInterface() {propertyValues {Curve {keys [
          <at 0 0 RCIM_Cubic>
          <at 0.2 1.3 RCIM_Cubic>
          {arriveTangent=-0.1423, interpMode=RCIM_Cubic, leaveTangent=-0.1423, tangentMode=RCTM_User, time=1, value=0}
         ]}}}
      Scale Curve=15
     )
     Vector2D=<0.7 1.5>
    )) {isSetParameters=true}
   /SetVariables(Particles.Glow03_Color=&MakeLinearColorFromVectorAndFloat(
     "Vector (RGB)"=&MakeVectorFromLinearColorRGB(Color=&MultiplyLinearColor(Linear Color=<rgba 1 0.0987 0.135 0.2>, Scale Factor=&MakeVector4_FromFloat(Float=1)))
     "Float (Alpha)"=&Multiply_Float(A=0.2, B=1)
    )) {isSetParameters=true}
  ]
  rendererClasses [&NiagaraSpriteRendererProperties]
  renderers [
   {rendererClass=&NiagaraSpriteRendererProperties}
   {rendererClass=&NiagaraSpriteRendererProperties, rendererIndex=1}
  ]
 }
 Feathers {
  emitterUpdateScript [
   &EmitterState(
    Inactive Response=::NewEnumerator1
    Life Cycle Mode=::NewEnumerator1
    Loop Behavior=::NewEnumerator1
    Loop Delay=&Add_Float(A=$User.Delay, B=0.1)
    Loop Duration Mode=::NewEnumerator1
    UseLoopDelay=true
   )
   &SpawnRate(SpawnRate=&Lerp_Float(A=10, B=20, Alpha=$User.ParaGlideLeanAlpha), Use Spawn Probability=false)
  ]
  particleSpawnScript [
   /SetVariables(Particles.VisibilityTag=&RandomRangeInt(Minimum=0, Maximum=1)) {isSetParameters=true}
   &PartitionParticles()
   &InitializeParticle(
    Lifetime=&MergeHelper_FloatSwitch(
     0=&RandomRangeFloat(Minimum=1, Maximum=1.8)
     1=&RandomRangeFloat(Minimum=0.8, Maximum=1.8)
     2=?
     3=?
     Index=$Particles.VisibilityTag
     Num=::NewEnumerator0
    )
    Mesh Scale Mode=::NewEnumerator3
    Sprite Rotation Mode=::NewEnumerator2
    Sprite Size=&Multiply_Vector2D_ByFloat(Float=1.5, Vector2D=&RandomRangeVector2D(Minimum=<7 15>, Maximum=<15 25>))
    Sprite Size Mode=::NewEnumerator4
    UsePositionOffset=true
    Write Material Random=true
   )
   &CameraOffset(Camera Offset Amount=30)
   &SkeletalMeshLocation(
    Bone / Socket Index=$Particles.PartitionParticles.Partition
    Bone Sampling Mode=::NewEnumerator4
    MaskSpawnGroup=false
    Orientation Sampling=::NewEnumerator1
    Position Sampling=::NewEnumerator1
    Sampled Position Offset=<0 0 20>
    Sampled Position Offset Coordinate Space:ENiagaraSkelMeshTransforms=::NewEnumerator1
    Skeletal Mesh=$System.Skeletal Mesh
   )
   &AddVelocity(
    Velocity=&Add_Vector(
     A=&RandomRangeVector(Minimum=<-5 -5 0>, Maximum=<5 5 15>)
     B=&Lerp_Vector(
      A=<0 -12 0>
      B=<0 -15 0>
      Alpha=&FN_FrontEndFloatModifier(Gameplay=0, Front End=$User.ParaGlideLeanAlpha)
     )
    )
    Velocity Speed Scale=&RandomRangeFloat(Minimum=5, Maximum=8)
   ) {module=AddVelocity001}
   &SubUVAnimation(
    Evaluation Type=::NewEnumerator0
    "Sprite Renderer "=DataInterface() {propertyValues {SpriteRenderer=&self:Feathers_0.NiagaraSpriteRendererProperties_0}}
    SubUV Animation Mode:ENiagara_SubUVLookupModeV2=::NewEnumerator2
    UseEndFrame=false
    UseStartFrame=false
   ) {module=SubUVAnimation001}
   &DynamicMaterialParameters(
    Index 1 Param 2=&RandomRangeFloat(Minimum=4, Maximum=6)
    Module.Index1Float0Write=true
    Module.Index1Float1Write=true
    Module.Index1Float2Write=true
    Module.Index1Float3Write=true
    Write Parameter Index 0=::NewEnumerator0
    Write Parameter Index 1=::NewEnumerator1
   )
  ]
  particleUpdateScript [
   &ParticleState()
   &SpriteRotationRate(Rotation Rate=&RandomRangeFloat(Minimum=-250, Maximum=250))
   &ScaleSpriteSize(
    Scale Sprite Size Mode=::NewEnumerator0
    Uniform Scale Factor=&FloatFromCurve(
     FloatCurve=DataInterface() {propertyValues {Curve {
        defaultValue=1
        keys [
         {arriveTangent=4.1785, interpMode=RCIM_Cubic, leaveTangent=4.1785, tangentMode=RCTM_User, time=0, value=0}
         {arriveTangent=1.1179, interpMode=RCIM_Cubic, leaveTangent=1.1179, tangentMode=RCTM_User, time=0.3, value=1.6}
         {arriveTangent=-2.2857, interpMode=RCIM_Cubic, leaveTangent=-2.2857, time=0.8, value=1.9906}
         {arriveTangent=0.0224, interpMode=RCIM_Cubic, leaveTangent=0.0224, tangentMode=RCTM_User, time=1, value=0}
        ]
       }}}
     Scale Curve=&Multiply_Float(
      A=0.2
      B=/Niagara/DynamicInputs/Camera/CameraFade(
       Camera Query:NiagaraDataInterfaceCamera=DataInterface()
       Fade Offset:NiagaraFloat=50
       Fade Distance:NiagaraFloat=150
       Use Owner Position:NiagaraBool=true
      )
     )
    )
   )
   &ScaleColor(
    Scale Alpha=&Multiply_Float(
     A=&RandomRangeFloat(Minimum=0.7, Maximum=1)
     B=&FloatFromCurve(
      FloatCurve=DataInterface() {propertyValues {Curve {keys [
          {arriveTangent=3.7485, interpMode=RCIM_Cubic, leaveTangent=3.7485, tangentMode=RCTM_User, time=0, value=0.35}
          {interpMode=RCIM_Cubic, tangentMode=RCTM_User, time=0.5, value=1}
          {tangentMode=RCTM_User, time=1, value=0}
         ]}}}
      Scale Curve=2
     )
    )
    Scale RGB=&MakeVectorFromLinearColorRGB(Color=&ColorFromCurve(
      ColorCurve=DataInterface() {propertyValues {
        AlphaCurve {keys [<at 0 1 RCIM_Cubic>]}
        BlueCurve {keys [<at 0.0515 0.3482 RCIM_Cubic>,<at 0.785 0.0273 RCIM_Cubic>]}
        GreenCurve {keys [<at 0.0515 0.1808 RCIM_Cubic>,<at 0.785 0.011 RCIM_Cubic>]}
        RedCurve {keys [<at 0.0515 1 RCIM_Cubic>,<at 0.785 0.0116 RCIM_Cubic>]}
       }}
      CurveIndex=&Add_Float(
       A=$Particles.Age
       B=&Lerp_Float(
        A=0
        B=0.3
        Alpha=&FloatFromCurve(FloatCurve=DataInterface() {propertyValues {Curve {keys [<at 0.7 0>,<at 1 1>]}}}, CurveIndex=$User.ParaGlideLeanAlpha)
       )
      )
     ))
    ScaleA=true
    ScaleRGB=true
   )
   &Drag(
    Drag=&MergeHelper_FloatSwitch(
     0=&RandomRangeFloat(Minimum=2, Maximum=3)
     1=&RandomRangeFloat(Minimum=0.5, Maximum=1)
     2=?
     3=?
     Index=$Particles.VisibilityTag
     Num=::NewEnumerator0
    )
    Use Linear Drag=true
    Use Rotational Drag=true
   )
   &GravityForce(Gravity=&Add_Vector(
     A=&RandomRangeVector(Minimum=<0 0 200>, Maximum=<0 0 600>)
     B=&Lerp_Vector(A=<0 0 0>, B=<0 0 500>, Alpha=$User.ParaGlideLeanAlpha)
    ))
   &CurlNoiseForce(
    Bias Noise=false
    Noise Frequency=55
    Noise Strength=&RandomRangeFloat(Minimum=200, Maximum=300)
    Pan Noise=false
   )
   &SolveForcesAndVelocity(Clamp Velocity=false, Limit Acceleration=false)
  ]
  rendererClasses [&NiagaraSpriteRendererProperties]
  renderers [
   {rendererClass=&NiagaraSpriteRendererProperties}
   {rendererClass=&NiagaraSpriteRendererProperties, rendererIndex=1}
  ]
 }
 Fire {
  emitterUpdateScript [
   &EmitterState(Delay First Loop Only=true, Life Cycle Mode=::NewEnumerator1, Loop Delay=$User.Delay, Loop Duration=1, UseLoopDelay=true)
   &SpawnParticlesFromOtherEmitter(
    Attribute Reader=DataInterface() {propertyValues {EmitterBinding {bindingMode=Other, emitterName=Soul}}}
    Spawn Rate=20
    Use Maximum Age=false
    Use Minimum Age=false
   )
  ]
  particleSpawnScript [
   /SetVariables(Particles.VisibilityTag=/Niagara/DynamicInputs/Divide/Divide_Int(
     A:NiagaraInt32=&ModuloInt(
      A=&Divide_Int(A:NiagaraInt32=&ReturnParticleIDAttributes(), B:NiagaraInt32=1)
      B=4
     )
     B:NiagaraInt32=2
    )) {isSetParameters=true}
   &InitializeParticle(
    Color=&MergeHelper_ColorSwitch(0=<rgba 1 1 1 2>, 1=<rgba 0.7812 0.7812 0.7812 0.5>, 2=?, 3=?, Index=$Particles.VisibilityTag, Num=::NewEnumerator0)
    Lifetime=&MergeHelper_FloatSwitch(
     0=&RandomRangeFloat(Minimum=0.4, Maximum=0.6)
     1=&RandomRangeFloat(Minimum=0.7, Maximum=1.2)
     2=?
     3=?
     Index=$Particles.VisibilityTag
     Num=::NewEnumerator0
    )
    Sprite Rotation Mode=::NewEnumerator2
    Sprite Size Mode=::NewEnumerator1
    Sprite UV Mode=::NewEnumerator1
    Uniform Sprite Size Max=80
    Uniform Sprite Size Min=50
    UsePositionOffset=false
    Write Material Random=true
   )
   &SampleParticlesFromOtherEmitter(Mesh Orientation Sampling:ENiagara_DirectReadSamplingMode=::NewEnumerator0, Transform Data Option=1)
   &CameraOffset(Camera Offset Amount=&RandomRangeFloat(Minimum=-8, Maximum=-3))
  ]
  particleUpdateScript [
   &ParticleState()
   &ShapeLocation(
    Additional Quaternion Rotation=false
    Apply To Particle Position=::NewEnumerator1
    MaskSpawnGroup=false
    Non Uniform Scale=<0.8 1 2>
    Offset Mode=::NewEnumerator0
    Override Seed=false
    Rotation Mode=::NewEnumerator3
    Rotation Quaternion=&MultiplyQuaternion(
     Quaternion A=&self:Get_MeshOrientation_Attribute_By_Index(
      Particle Attribute Reader:NiagaraDataInterfaceParticleRead=DataInterface() {propertyValues {EmitterBinding {bindingMode=Other, emitterName=Soul}}}
      Particle Index:NiagaraInt32=&MakeIntFromFloat(Float=$Particles.SampleParticlesFromOtherEmitter.SampledIndex)
     )
     Quaternion B=&FindQuatBetweenTwoVectors(From Vector=<1 0 0>, To Vector=<1 0 1>)
    )
    Sphere Radius=8
    Transform Order:ENiagara_TransformOrder=::NewEnumerator1
    UseOwnerScale=true
   )
   &InheritSourceMovement(
    Applied Position Scale=&VectorFromFloat(Value=&FloatFromCurve(FloatCurve=DataInterface() {propertyValues {Curve {keys [<at 0 1>,<at 1 0.95>]}}}))
    ApplyForce=false
    ApplyPosition=true
    ApplyPositionFalloff=false
   )
   &SolveForcesAndVelocity(Clamp Velocity=false, Limit Acceleration=false)
   &ScaleSpriteSize(Scale Factor=&Vector2DFromFloat(Value=&FloatFromCurve(FloatCurve=DataInterface() {propertyValues {Curve {keys [
          {arriveTangent=2.5727, interpMode=RCIM_Cubic, leaveTangent=2.5727, tangentMode=RCTM_User, time=0, value=0.1}
          {interpMode=RCIM_Cubic, leaveTangent=-0.8944, time=1, value=0.8}
         ]}}})))
   &DynamicMaterialParameters(
    Index 0 Param 1=&FloatFromCurve(FloatCurve=DataInterface() {propertyValues {Curve {keys [{arriveTangent=2.7551, interpMode=RCIM_Cubic, leaveTangent=2.7551, tangentMode=RCTM_User, time=0, value=0.5}, <at 0.5 1>]}}})
    Index 0 Param 2=&RandomRangeFloat(Minimum=0.33, Maximum=1)
    Index 0 Param 3=&FloatFromCurve(FloatCurve=DataInterface() {propertyValues {Curve {keys [<at 0 1.25>,<at 1 2>]}}})
    Index 0 Param 4=1
    Module.Param0WriteEnabled=true
    Module.Param1WriteEnabled=true
    Module.Param2WriteEnabled=true
    Module.Param3WriteEnabled=true
   )
   &ScaleColor(
    Scale Mode=::NewEnumerator1
    Scale RGBA=&Multiply_Vector4(
     A=/Game/Effects/Niagara/Modules/MergeEmitters/MergeHelper_Vector4Switch(
      0:Vector4f=&Vector4FromCurve(Vector4Curve=DataInterface() {propertyValues {
         AlphaCurve {keys [<at 0.8663 1>,<at 1 0>]}
         BlueCurve {keys [<at 0 1.2487>,<at 0.2603 0.8625>,<at 1 0.0857>]}
         GreenCurve {keys [<at 0 1.3442>,<at 0.2603 0.4531>,<at 1 0.0333>]}
         RedCurve {keys [<at 0 5>,<at 0.2603 3>,<at 1 0.3594>]}
        }})
      1:Vector4f=&Vector4FromCurve(Vector4Curve=DataInterface() {propertyValues {
         AlphaCurve {keys [<at 0 1>,<at 0.6619 0.1>,<at 0.8586 0>]}
         BlueCurve {keys [<at 0.0106 0.3449>,<at 0.2594 0.2077>,<at 0.9271 0.0921>]}
         GreenCurve {keys [<at 0.0106 0.2826>,<at 0.2594 0.1094>,<at 0.9271 0.0629>]}
         RedCurve {keys [<at 0.0106 0.5833>,<at 0.2594 0.5833>,<at 0.9271 0.2344>]}
        }})
      2:Vector4f=?
      3:Vector4f=?
      4:Vector4f=?
      5:Vector4f=?
      6:Vector4f=?
      7:Vector4f=?
      8:Vector4f=?
      9:Vector4f=?
      Index:NiagaraInt32=$Particles.VisibilityTag
      Num:MergeHelper_SwitcherEnum=::NewEnumerator0
     )
     B=&MakeVector4FromVectorAndFloat("Vector (RGB)"=&Lerp_Vector(A=<1 1 1>, B=<1.5 1.5 1.5>, Alpha=$User.ParaGlideLeanAlpha), "Float (Alpha)"=1)
    )
    ScaleRGBA=true
   )
   &SubUVAnimation(
    "Sprite Renderer "=DataInterface() {propertyValues {SpriteRenderer=&self:Fire_7.NiagaraSpriteRendererProperties_2}}
    Start Frame Range Override=&RandomRangeInt(Minimum=0, Maximum=20)
    UseEndFrame=false
    UseStartFrame=false
   )
   &SubUVAnimation(
    End Frame Range Override=64
    "Sprite Renderer "=DataInterface() {propertyValues {SpriteRenderer=&self:Fire_7.NiagaraSpriteRendererProperties_0}}
    UseEndFrame=false
    UseStartFrame=false
   ) {module=SubUVAnimation001}
   /SetVariables(Particles.Position=/Niagara/DynamicInputs/Lerp/Lerp_Position(
     A:NiagaraPosition=&Add_Vector(
      A=&Add_Vector(
       A=&self:Get_Position_Attribute_By_Index(
        Particle Attribute Reader:NiagaraDataInterfaceParticleRead=DataInterface() {propertyValues {EmitterBinding {bindingMode=Other, emitterName=Soul}}}
        Particle Index:NiagaraInt32=&MakeIntFromFloat(Float=$Particles.SampleParticlesFromOtherEmitter.SampledIndex)
       )
       B=$Particles.ShapeLocation.ShapeVector
      )
      B=<0 0 0>
     )
     B:NiagaraPosition=&AddVectorToPosition(
      Position=$Particles.Position
      Vector=&RotateVectorByQuaternion(
       Vector To Rotate=&MergeHelper_VectorSwitch(0=<0 0 5>, 1=<0 0 20>, 2=?, 3=?, Index=$Particles.VisibilityTag, Num=::NewEnumerator0)
       Quaternion=&MultiplyQuaternion(
        Quaternion A=&self:Get_MeshOrientation_Attribute_By_Index(
         Particle Attribute Reader:NiagaraDataInterfaceParticleRead=DataInterface() {propertyValues {EmitterBinding {bindingMode=Other, emitterName=Soul}}}
         Particle Index:NiagaraInt32=&MakeIntFromFloat(Float=$Particles.SampleParticlesFromOtherEmitter.SampledIndex)
        )
        Quaternion B=&FindQuatBetweenTwoVectors(From Vector=<1 0 0>, To Vector=<1 0 1>)
       )
      )
     )
     Alpha:NiagaraFloat=&FloatFromCurve(FloatCurve=DataInterface() {propertyValues {Curve {keys [
          {arriveTangent=1.9048, interpMode=RCIM_Cubic, leaveTangent=1.9048, tangentMode=RCTM_User, time=0.1, value=0}
          {arriveTangent=0.5863, interpMode=RCIM_Cubic, leaveTangent=0.5863, tangentMode=RCTM_User, time=1, value=1}
         ]}}})
    )) {isSetParameters=true}
   /SetVariables(
    Particles.SmokeDynamic:Vector4f=&MakeVector4(
     X=&FloatFromCurve(FloatCurve=DataInterface() {propertyValues {Curve {keys [<at 0.5 0.05 RCIM_Cubic>, {interpMode=RCIM_Cubic, leaveTangent=4.7896, time=1, value=0.2}]}}})
     Y=&FloatFromCurve(FloatCurve=DataInterface() {propertyValues {Curve {keys [
          {interpMode=RCIM_Cubic, tangentMode=RCTM_User, time=0, value=0}
          {interpMode=RCIM_Cubic, tangentMode=RCTM_User, time=0.5, value=1}
          {tangentMode=RCTM_User, time=1, value=0}
         ]}}})
     Z=&RandomRangeFloat(Minimum=0.6, Maximum=0.8)
     W=150
    )
    Particles.SmokeDynamic001:Vector4f=&MakeVector4(X=150, Y=200, Z=0, W=0)
   ) {isSetParameters=true}
   &KillParticles(Kill Particles=&SetBoolByFloatComparison(
     A=&MergeHelper_FloatSwitch(0=0, 1=$Particles.NormalizedAge, 2=?, 3=?, Index=$Particles.VisibilityTag, Num=::NewEnumerator0)
     B=0.8
    ))
  ]
  rendererClasses [&NiagaraSpriteRendererProperties]
  renderers [
   {rendererClass=&NiagaraSpriteRendererProperties}
   {rendererClass=&NiagaraSpriteRendererProperties, rendererIndex=1}
  ]
 }
 FireTrail_And_Dots {
  emitterUpdateScript [
   &EmitterState(
    Inactive Response=::NewEnumerator1
    Life Cycle Mode=::NewEnumerator1
    Loop Behavior=::NewEnumerator1
    Loop Duration Mode=::NewEnumerator1
    UseLoopDelay=false
   )
   &SpawnBurst_Instantaneous(Spawn Count=2, Use Spawn Probability=false, UseLoopCountLimit=false)
  ]
  particleSpawnScript [
   &FN_SetSpawnGroup(Number of Spawn Groups=2)
   /SetVariables(Particles.BoneIndex:NiagaraInt32=&Divide_Int(A:NiagaraInt32=&ReturnExecIndex(), B:NiagaraInt32=2)) {isSetParameters=true}
   &InitializeParticle(
    Color=<rgba 1 1 1 2>
    Lifetime=3
    Mesh Scale=&RandomRangeVector(Minimum=<1 1 2.6>, Maximum=<1 1 3>)
    Mesh Scale Mode=::NewEnumerator4
    UsePositionOffset=false
    Write Material Random=true
   )
   &InitialMeshOrientation(
    Mesh Axis To Orient=<0 0 1>
    Mesh Orientation Mode=::NewEnumerator2
    Orientation Coordinate Space=::World
    Orientation Vector=&FN_FrontEndVector3Modifier(Gameplay=<0 0 1>, Front End=<0 -0.3 1>)
    Use Rotation Vector=false
   )
   &self:SampleSourceSpawnGroupFromOtherParticle_02(PAR=DataInterface() {propertyValues {EmitterBinding {bindingMode=Other, emitterName=Soul}}}) {module=SampleSourceSpawnGroupFromOtherParticle_02}
   &CameraOffset(Camera Offset Amount=-20)
   /SetVariables() {isSetParameters=true}
   /SetVariables() {isSetParameters=true}
   /SetVariables(
    Particles.DotDyn003:Vector4f=&MakeVector4(X=0, Y=&RandomRangeFloat(Minimum=0, Maximum=100), Z=0, W=0)
    Particles.DotDyn_1:Vector4f=&MakeVector4(X=0, Y=&RandomRangeFloat(Minimum=0.9, Maximum=1.1), Z=0.2, W=0)
   ) {isSetParameters=true}
  ]
  particleUpdateScript [
   &ParticleState(Kill Particles When Lifetime Has Elapsed=false, Let Infinitely Lived Particles Die When Emitter Deactivates=true)
   &ScaleMeshSize(Scale Factor=&MakeVector(
     X=1
     Y=1
     Z=&FloatFromCurve(
      FloatCurve=DataInterface() {propertyValues {Curve {keys [<at 0.5 0 RCIM_Cubic>, {arriveTangent=6.2501, interpMode=RCIM_Cubic, leaveTangent=6.2501, tangentMode=RCTM_User, time=1, value=1}]}}}
      CurveIndex=$Engine.System.Age
      Scale Curve=&Lerp_Float(A=1, B=1.2, Alpha=$User.ParaGlideLeanAlpha)
     )
    )) {module=ScaleMeshSize001}
   &self:SamplePositionFromOtherParticle_01(PAR=DataInterface() {propertyValues {EmitterBinding {bindingMode=Other, emitterName=Soul}}}) {module=SamplePositionFromOtherParticle_01}
   /SetVariables(Particles.SkullScale=1) {isSetParameters=true}
   /SetVariables(Particles.Position=&AddVectorToPosition(
     Position=$Particles.Position
     Vector=&TransformVector(
      Vector=&Multiply_VectorByFloat(
       Vector=&FN_FrontEndVector3Modifier(
        Gameplay=&Lerp_Vector(A=<0 0 -25>, B=<0 0 -25>, Alpha=$System.InterpolateOverTime.InterpolateOverTime_Float.Moving Average)
        Front End=<0 10 -25>
       )
       Float=$Particles.SkullScale
      )
      Destination Space=::Simulation
     )
    )) {isSetParameters=true}
   &InterpolateOverTime(Rate Of Change=5, Target Vector Value:NiagaraPosition=$Particles.Position, Variable Type=::NewEnumerator2)
   /SetVariables(Particles.TrailDynOffset:NiagaraFloat=&Add_Float(
     A=$Particles.TrailDynOffset
     B=&Multiply_Float(A=$Engine.DeltaTime, B=&Lerp_Float(A=0, B=-0.5, Alpha=$User.ParaGlideLeanAlpha))
    )) {isSetParameters=true}
   &DynamicMaterialParameters(
    Index 0 Param 1=&Multiply_Float(
     A=&FN_FrontEndFloatModifier(
      Gameplay=&Lerp_Float(A=110, B=110, Alpha=$System.InterpolateOverTime.InterpolateOverTime_Float.Moving Average)
      Front End=&Lerp_Float(A=80, B=80, Alpha=$System.InterpolateOverTime.InterpolateOverTime_Float.Moving Average)
     )
     B=&FloatFromCurve(
      FloatCurve=DataInterface() {propertyValues {Curve {keys [<at 0.5 0 RCIM_Cubic>, {interpMode=RCIM_Cubic, leaveTangent=3.362, time=1.5, value=1}]}}}
      CurveIndex=$System.Age
     )
    )
    Index 0 Param 2=&Multiply_Float(
     A=&FN_FrontEndFloatModifier(
      Gameplay=&Lerp_Float(A=150, B=150, Alpha=$System.InterpolateOverTime.InterpolateOverTime_Float.Moving Average)
      Front End=&Lerp_Float(A=130, B=130, Alpha=$System.InterpolateOverTime.InterpolateOverTime_Float.Moving Average)
     )
     B=&FloatFromCurve(
      FloatCurve=DataInterface() {propertyValues {Curve {keys [<at 0.5 0 RCIM_Cubic>, {arriveTangent=1.9352, interpMode=RCIM_Cubic, leaveTangent=1.9352, tangentMode=RCTM_User, time=1.5, value=1}]}}}
      CurveIndex=$System.Age
     )
    )
    Index 0 Param 3=1
    Index 0 Param 4=&Add_Float(A=&RandomRangeFloat(Minimum=0, Maximum=100), B=$Particles.TrailDynOffset)
    Index 1 Param 1=&Lerp_Float(A=0, B=1, Alpha=$System.InterpolateOverTime.InterpolateOverTime_Float.Moving Average)
    Index 1 Param 2=&FloatFromCurve(FloatCurve=DataInterface() {propertyValues {Curve {keys [{arriveTangent=3.6042, interpMode=RCIM_Cubic, time=0.1, value=0}, <at 0.8 1 RCIM_Cubic>]}}})
    Index 1 Param 3=&FN_FrontEndFloatModifier(
     Gameplay=0
     Front End=&Lerp_Float(A=-70, B=-10, Alpha=$System.InterpolateOverTime.InterpolateOverTime_Float.Moving Average)
    )
    Index 1 Param 4=&FN_FrontEndFloatModifier(
     Gameplay=&Lerp_Float(A=1, B=1, Alpha=$System.InterpolateOverTime.InterpolateOverTime_Float.Moving Average)
     Front End=1
    )
    Index 2 Param 1=1
    Index 2 Param 2=&Lerp_Float(
     A=50
     B=50
     Alpha=&ClampFloat(Float=&Add_Float(
       A=&RemapRange(Input Value To Remap=$Particles.RotAmount, Input Min=15, Input Max=90, Output Min=0.1, Output Max=1)
       B=$System.Timeline004.NormalizedTime
      ))
    )
    Index 2 Param 3=&FloatFromCurve(
     FloatCurve=DataInterface() {propertyValues {Curve {keys [<at 0.4 -1>,<at 2.5 1>]}}}
     CurveIndex=&Add_Float(A=$Emitter.Age, B=&RandomRangeFloat(Minimum=-0.3, Maximum=0))
    )
    Index 3 Param 1=&MakeFloatFromVector(Vector=$StackContext.InterpolateOverTime.InterpolateOverTime_V3_.Moving Average, Channel=::NewEnumerator0)
    Index 3 Param 2=&MakeFloatFromVector(Vector=$StackContext.InterpolateOverTime.InterpolateOverTime_V3_.Moving Average, Channel=::NewEnumerator1)
    Index 3 Param 3=&MakeFloatFromVector(Vector=$StackContext.InterpolateOverTime.InterpolateOverTime_V3_.Moving Average, Channel=::NewEnumerator2)
    Index 3 Param 4=1
    Module.Index0Param0Write=true
    Module.Index0Param1Write=true
    Module.Index0Param2Write=true
    Module.Index0Param3Write=true
    Module.Index1Param0Write=true
    Module.Index1Param1Write=true
    Module.Index1Param2Write=true
    Module.Index1Param3Write=true
    Module.Index2Param0Write=true
    Module.Index2Param1Write=true
    Module.Index2Param2Write=true
    Module.Index2Param3Write=false
    Module.Index3Param0Write=true
    Module.Index3Param1Write=true
    Module.Index3Param2Write=true
    Module.Index3Param3Write=true
    Write Parameter Index 1=true
    Write Parameter Index 2=true
    Write Parameter Index 3=true
   )
   &ScaleColor(
    Scale RGB=&MakeVectorFromLinearColorRGB(Color=&LerpLinearColors(StartColor=<rgba 1 1 1 1>, EndColor=<rgba 0.75 0.525 0.525 1>, LerpFactor=$User.ParaGlideLeanAlpha))
    ScaleA=true
    ScaleRGB=true
   )
   &UpdateMeshOrientation(
    Facing Interpolation=true
    Facing Interpolation Rate=3
    Facing Position:NiagaraPosition=&AddVectorToPosition(
     Position=$Particles.Position
     Vector=&TransformVector(
      Vector=&FN_FrontEndVector3Modifier(Gameplay=<0 0 1>, Front End=<0 -0.45 1>)
      Destination Space=::Simulation
     )
    )
    Orientation Method=::NewEnumerator2
    Override Incoming Orientation=1
    Override Quat=true
    Override Quaternion:Quat4f=$Particles.DotsOrientation
    TransformFacingOrigin=false
    TransformFacingPosition=false
    TransformUp=false
    Up Direction:Vector3f=&NormalizeVector(
     Vector To Normalize=&Subtract_Position(
      A=&ReturnCameraPosition(
       Coordinate Space:ENiagaraCoordinateSpace=::Simulation
       Camera Query:NiagaraDataInterfaceCamera=DataInterface() {propertyValues {bRequireCurrentFrameData=false}}
      )
      B=$Particles.Position
     )
     Safe Normalize=true
     Safe Fallback Threshold=0
     Safe Fallback Vector=<0 0 0.001>
    )
    Write to Orientation:ENiagara_AttributeSamplingApplyOutput=::NewEnumerator1
   ) {module=UpdateMeshOrientation002}
   /SetVariables(Particles.DotsOrientation:Quat4f=$Output.UpdateMeshOrientation002.MeshOrientation) {isSetParameters=true}
   /SetVariables(Particles.MeshOrientation=&MultiplyQuaternion(Quaternion A=$Engine.Owner.Rotation, Quaternion B=$Particles.Initial.MeshOrientation)) {isSetParameters=true}
   /SetVariables(
    Particles.DotColor:LinearColor=&MultiplyLinearColor(
     Linear Color=$Particles.Initial.Color
     Scale Factor=&MakeVector4(
      X=1
      Y=1
      Z=1
      W=&FloatFromCurve(FloatCurve=DataInterface() {propertyValues {Curve {keys [<at 0.5 0>,<at 0.8 0.5>]}}}, Scale Curve=0.8)
     )
    )
    Particles.DotPosition:NiagaraPosition=&AddVectorToPosition(
     Position=$Particles.Position
     Vector=&TransformVector(
      Vector=&Multiply_VectorByFloat(
       Vector=&FN_FrontEndVector3Modifier(
        Gameplay=&Lerp_Vector(A=<0 0 25>, B=<0 0 25>, Alpha=$System.InterpolateOverTime.InterpolateOverTime_Float.Moving Average)
        Front End=<5 -10 25>
       )
       Float=$Particles.SkullScale
      )
      Destination Space=::Simulation
     )
    )
   ) {isSetParameters=true}
  ]
  rendererClasses [&NiagaraMeshRendererProperties]
  renderers [
   {rendererClass=&NiagaraMeshRendererProperties, rendererIndex=0}
   {rendererClass=&NiagaraMeshRendererProperties, rendererIndex=1}
  ]
 }
 Glow {
  emitterUpdateScript [
   &EmitterState(
    Life Cycle Mode=::NewEnumerator1
    Loop Behavior=::NewEnumerator1
    Loop Delay=&Add_Float(A=$User.Delay, B=0.1)
    Loop Duration Mode=::NewEnumerator1
    UseLoopDelay=true
   )
   &SpawnBurst_Instantaneous(Spawn Count=2, Use Spawn Probability=false, UseLoopCountLimit=false)
   &SpawnBurst_Instantaneous(Spawn Count=2, Spawn Group=1, Use Spawn Probability=false, UseLoopCountLimit=false) {module=SpawnBurst_Instantaneous001}
   &SpawnRate(SpawnGroup=2, SpawnRate=15, Use Spawn Probability=false)
  ]
  particleSpawnScript [
   /SetVariables(Particles.VisibilityTag=$Emitter.SpawnGroup) {isSetParameters=true}
   &PartitionParticles()
   &InitializeParticle(
    Color=<rgba 0.8281 0.0259 0.0553 0.774>
    Lifetime=&MergeHelper_FloatSwitch(0=0.5, 1=0.5, 2=0.7, 3=?, Index=$Particles.VisibilityTag, Num=::NewEnumerator1)
    Mesh Scale Mode=::NewEnumerator1
    Mesh Uniform Scale Max=0.7
    Mesh Uniform Scale Min=0.6
    Sprite Rotation Mode=::NewEnumerator2
    Sprite Size Mode=::NewEnumerator3
    Uniform Sprite Size=&MergeHelper_FloatSwitch(0=100, 1=80, 2=0, 3=?, Index=&SpawnGroup(), Num=::NewEnumerator1)
    UsePositionOffset=false
    Write Material Random=true
   )
   &InitialMeshOrientation(Orientation Axis=<0 0 1>, Orientation Vector=<0 0 1>, Rotation=<0 0.25 0>, Use Orientation Vector=true, Use Rotation Vector=true)
   &CameraOffset(Camera Offset Amount=&MergeHelper_FloatSwitch(0=5, 1=5, 2=15, 3=?, Index=$Particles.VisibilityTag, Num=::NewEnumerator1))
   &AddVelocity(
    Velocity=<0 0 1>
    Velocity Speed Scale=&MergeHelper_FloatSwitch(
     0=0
     1=0
     2=&RandomRangeFloat(Minimum=10, Maximum=20)
     3=?
     Index=$Particles.VisibilityTag
     Num=::NewEnumerator1
    )
   )
   &InitialMeshOrientation(
    Mesh Orientation Mode=::NewEnumerator7
    Rotation=&RandomRangeVector(Minimum=<0 0 0>, Maximum=<1 0 0>)
    Use Rotation Vector=true
   ) {module=InitialMeshOrientation001}
  ]
  particleUpdateScript [
   &ParticleState(Kill Particles When Lifetime Has Elapsed=false, Let Infinitely Lived Particles Die When Emitter Deactivates=true)
   &ScaleSpriteSize(
    Scale Sprite Size Mode=::NewEnumerator0
    Uniform Scale Factor=&Lerp_Float(A=1, B=1.2, Alpha=$User.ParaGlideLeanAlpha)
   )
   &SkeletalMeshLocation(
    Bone / Socket Index=$Particles.PartitionParticles.Partition
    Bone Sampling Mode=::NewEnumerator4
    MaskSpawnGroup=false
    Orientation Sampling=::NewEnumerator1
    Position Sampling=::NewEnumerator1
    Sampled Position Offset=<0 0 25>
    Sampled Position Offset Coordinate Space:ENiagaraSkelMeshTransforms=::NewEnumerator2
    Skeletal Mesh=$System.Skeletal Mesh
   ) {module=SkeletalMeshLocation002}
   &ScaleColor(
    Scale Alpha=&FloatFromCurve(
     FloatCurve=DataInterface() {propertyValues {Curve {keys [<at 0 0>,<at 1 1>]}}}
     Scale Curve=&Multiply_Float(
      A=&FloatFromCurve(
       FloatCurve=DataInterface() {propertyValues {Curve {keys [{interpMode=RCIM_Cubic, leaveTangent=3.3849, tangentMode=RCTM_Break, time=0, value=0}, <at 1 1>]}}}
       Scale Curve=&Waveform(
        Global Input Position=&RandomRangeFloat(Minimum=0, Maximum=100)
        "Waveform [1]"=::NewEnumerator9
        "[1] Frequency"=&RandomRangeFloat(Minimum=8, Maximum=12)
        "[1] Amplitude Min/Max"=<0.5 1>
       )
      )
      B=&FN_FrontEndFloatModifier(Gameplay=1, Front End=0.2)
     )
    )
    Scale RGB=&Multiply_VectorByFloat(
     Vector=&VectorFromFloat(Value=&Waveform(
       Waveform Count=::NewEnumerator1
       "Waveform [1]"=::NewEnumerator9
       "[1] Frequency"=3.5
       "[1] Amplitude Min/Max"=<0 0.6>
       "Waveform [2] Blend"=::NewEnumerator2
       "Waveform [2]"=::NewEnumerator9
       "[2] Input Position"=0
       "[2] Phase"=$Emitter.Age
       "[2] Phase Scale"=1
       "[2] Compound Frequency"=2
       "[2] Frequency"=1.4
       "[2] Amplitude Min/Max"=<0.2 1>
       "[2] Scale"=1
       "[2] Bias"=0
       "[2] Exponent"=1
       "[2] Pulse Duration"=0.1
       "[2] Clip Triangle Min/Max"=<0 0>
       "[2] Random Seed"=0
      ))
     Float=&FNTimeOfDayScalar(Night Scale=0.3, Morning Scale=1, Evening Scale=1)
    )
    ScaleA=true
    ScaleRGB=true
   )
   &FN_Light_Attributes(
    Color=$Particles.Color
    Create Color Attribute=true
    Create Exponent Attribute=true
    Create Position Attribute=true
    Create Radius Attribute=true
    Exponent=1
    Position=&AddVectorToPosition(Position=$Particles.Position, Vector=<0 0 -10>)
    Radius=30
    Scale Color=&FloatFromCurve(FloatCurve=DataInterface() {propertyValues {Curve {keys [<at 0 0>,<at 0.2 1>]}}}, Scale Curve=100)
   )
   &DynamicMaterialParameters(
    Index 0 Param 1=&MergeHelper_FloatSwitch(
     0=2
     1=0
     2=&FloatFromCurve(FloatCurve=DataInterface() {propertyValues {Curve {keys [{arriveTangent=2.8297, interpMode=RCIM_Cubic, time=0, value=-0.1}, <at 1 1 RCIM_Cubic>]}}})
     3=?
     Index=$Particles.VisibilityTag
     Num=::NewEnumerator1
    )
    Index 0 Param 2=&MergeHelper_FloatSwitch(0=0, 1=0, 2=1, 3=?, Index=$Particles.VisibilityTag, Num=::NewEnumerator1)
    Index 0 Param 3=&MergeHelper_FloatSwitch(0=0, 1=0, 2=1, 3=?, Index=$Particles.VisibilityTag, Num=::NewEnumerator1)
    Index 1 Param 1=&MergeHelper_FloatSwitch(
     0=&MergeHelper_FloatSwitch(0=-2, 1=2, 2=?, 3=?, Index=$Particles.PartitionParticles.Partition, Num=::NewEnumerator0)
     1=0
     2=0
     3=?
     Index=$Particles.VisibilityTag
     Num=::NewEnumerator1
    )
    Module.Index0Param0Write=true
    Module.Index0Param1Write=true
    Module.Index0Param2Write=true
    Module.Index0Param3Write=true
    Module.Index1Param0Write=true
    Module.Index1Param1Write=true
    Module.Index1Param2Write=true
    Module.Index1Param3Write=true
    Write Parameter Index 1=true
   )
   /SetVariables(Particles.DecalSize:Vector2f=&Multiply_Vector2D_ByFloat(
     Float=&MakeFloatFromVector(
      Vector=&VectorFromFloat(Value=&Waveform(
        Waveform Count=::NewEnumerator1
        "Waveform [1]"=::NewEnumerator9
        "[1] Phase Scale"=2
        "[1] Frequency"=3.5
        "[1] Amplitude Min/Max"=<0.8 1.2>
        "Waveform [2] Blend"=::NewEnumerator2
        "Waveform [2]"=::NewEnumerator7
        "[2] Input Position"=0
        "[2] Phase"=$Emitter.Age
        "[2] Phase Scale"=1
        "[2] Compound Frequency"=2
        "[2] Frequency"=1.4
        "[2] Amplitude Min/Max"=<1 1.3>
        "[2] Scale"=1
        "[2] Bias"=0
        "[2] Exponent"=1
        "[2] Pulse Duration"=0.1
        "[2] Clip Triangle Min/Max"=<0 0>
        "[2] Random Seed"=0
       ))
      Channel=::NewEnumerator0
     )
     Vector2D=<70 70>
    )) {isSetParameters=true}
   /SetVariables(Particles.GlowPosition:NiagaraPosition=&AddVectorToPosition(Position=$Particles.Position, Vector=<0 0 1.0151>)) {isSetParameters=true}
   /SetVariables(Particles.SpriteFacing=<0 -1 0>) {isSetParameters=true}
   /SetVariables(Particles.GlowColor=&MakeLinearColorFromVectorAndFloat(
     "Vector (RGB)"=&MakeVectorFromLinearColorRGB(Color=&LerpLinearColors(StartColor=<rgba 1.5 0.2272 0.3997 1>, EndColor=<rgba 1.5 0.2272 0.3997 1>, LerpFactor=$User.ParaGlideLeanAlpha))
     "Float (Alpha)"=&Multiply_Float(
      A=&FloatFromCurve(
       FloatCurve=DataInterface() {propertyValues {Curve {keys [{interpMode=RCIM_Cubic, leaveTangent=3.3849, tangentMode=RCTM_Break, time=0, value=0}, <at 1 1>]}}}
       Scale Curve=&Waveform(
        Global Input Position=&RandomRangeFloat(Minimum=0, Maximum=100)
        "Waveform [1]"=::NewEnumerator9
        "[1] Frequency"=&RandomRangeFloat(Minimum=8, Maximum=12)
        "[1] Amplitude Min/Max"=<1 2>
       )
      )
      B=&FN_FrontEndFloatModifier(Gameplay=1, Front End=0.2)
     )
    )) {isSetParameters=true}
   &ScaleMeshSize(Scale Factor=&VectorFromFloat(Value=&FloatFromCurve(FloatCurve=DataInterface() {propertyValues {Curve {keys [<at 0 0.2>,<at 1 1.2>]}}})))
   &KillParticles(Kill Particles=&SetBoolByFloatComparison(
     A=&MergeHelper_FloatSwitch(0=0, 1=0, 2=$Particles.NormalizedAge, 3=?, Index=$Particles.VisibilityTag, Num=::NewEnumerator1)
     B=0.6
     Comparison Type=::NewEnumerator2
    ))
   /SetVariables(Particles.ShockwaveColor:LinearColor=&ColorFromCurve(ColorCurve=DataInterface() {propertyValues {
       AlphaCurve {keys [<at 0 0.5>,<at 0.3576 0.5>,<at 0.5644 0>]}
       BlueCurve {keys [<at 0.0131 0.9644>,<at 0.6265 0.0192>,<at 0.9709 0.0052>]}
       GreenCurve {keys [<at 0.0131 0.5819>,<at 0.6265 0.0152>,<at 0.9709 0.0014>]}
       RedCurve {keys [<at 0.0131 5>,<at 0.6265 0.1823>,<at 0.9709 0.0035>]}
      }})) {isSetParameters=true}
   &SolveForcesAndVelocity(Clamp Velocity=false, Limit Acceleration=false)
  ]
  rendererClasses [
   &NiagaraLightRendererProperties
   &NiagaraSpriteRendererProperties
   &NiagaraMeshRendererProperties
  ]
  renderers [
   {rendererClass=&NiagaraLightRendererProperties, rendererIndex=0}
   {rendererClass=&NiagaraSpriteRendererProperties, rendererIndex=1}
   {rendererClass=&NiagaraSpriteRendererProperties, rendererIndex=2}
   {rendererClass=&NiagaraMeshRendererProperties, rendererIndex=3}
  ]
 }
 HornTrails {
  emitterUpdateScript [
   &EmitterState(Delay First Loop Only=true, Life Cycle Mode=::NewEnumerator1, Loop Delay=$User.Delay, Loop Duration=1, UseLoopDelay=true)
   &SpawnRate(SpawnRate=50, Use Spawn Probability=false)
  ]
  particleSpawnScript [
   &PartitionParticles(Number Of Partitions=4, Write Partition to Ribbon ID=::NewEnumerator0)
   &InitializeParticle(
    Color=<rgba 0.0521 0.0182 0.0241 1>
    Lifetime=1.5
    Ribbon Width=15
    Ribbon Width Mode=::NewEnumerator1
    UsePositionOffset=true
    Write Material Random=true
   )
  ]
  particleUpdateScript [
   &ParticleState()
   /SetVariables(Particles.Interpolate Dynamic=&MakeFloat4FromVectorAndFloat("Vector (RGB)"=<1 1 0>, "Float (Alpha)"=&FN_FrontEndFloatModifier(Gameplay=1, Front End=1))) {isSetParameters=true}
   &SolveForcesAndVelocity(Clamp Velocity=false, Limit Acceleration=false)
   &ScaleRibbonWidth(Ribbon Width Scale=&FloatFromCurve(
     FloatCurve=DataInterface() {propertyValues {Curve {keys [{arriveTangent=1.5831, interpMode=RCIM_Cubic, time=0, value=0.3}, {interpMode=RCIM_Cubic, leaveTangent=1.9068, time=1, value=1}]}}}
     Scale Curve=&FloatFromCurve(
      FloatCurve=DataInterface() {propertyValues {Curve {keys [<at 0.2 0 RCIM_Cubic>, {arriveTangent=3.0799, interpMode=RCIM_Cubic, leaveTangent=3.0799, tangentMode=RCTM_User, time=1, value=1}]}}}
      CurveIndex=$Engine.System.Age
     )
    ))
   &ScaleColor(
    Scale Alpha=&FloatFromCurve(FloatCurve=DataInterface() {propertyValues {Curve {keys [<at 0.2 1.5>,<at 1 0>]}}})
    Scale RGB=&MakeVectorFromLinearColorRGB(Color=&MultiplyLinearColor(
      Linear Color=&ColorFromCurve(ColorCurve=DataInterface() {propertyValues {
         AlphaCurve {keys [<at 0 1 RCIM_Cubic>]}
         BlueCurve {keys [<at 0.1883 3.6856 RCIM_Cubic>,<at 0.6134 1 RCIM_Cubic>]}
         GreenCurve {keys [<at 0.1883 3.1026 RCIM_Cubic>,<at 0.6134 1 RCIM_Cubic>]}
         RedCurve {keys [<at 0.1883 5 RCIM_Cubic>,<at 0.6134 1 RCIM_Cubic>]}
        }})
      Scale Factor=&MakeVector4FromVectorAndFloat("Vector (RGB)"=&Lerp_Vector(A=<1 1 1>, B=<1.5 1 1>, Alpha=$User.ParaGlideLeanAlpha), "Float (Alpha)"=1)
     ))
    ScaleA=true
    ScaleRGB=true
   )
   &self:SamplePosFromOtherParticle(
    Index:NiagaraInt32=&ModuloInt(A=$Particles.PartitionParticles.Partition, B=2)
    PAR=DataInterface() {propertyValues {EmitterBinding {bindingMode=Other, emitterName=Soul}}}
   ) {module=SamplePosFromOtherParticle001}
   /SetVariables(Particles.Position=&AddVectorToPosition(
     Position=&TransformPosition(Position=$Output.SamplePosFromOtherParticle001.ParticlePos, Source Space=::World, Destination Space=::World)
     Vector=&Add_Vector(
      A=&Lerp_Vector(
       A=<0 0 0>
       B=&Add_Vector(
        A=&MakeVector(
         X=&Waveform(
          Global Input Position=&MergeHelper_FloatSwitch(0=1.2, 1=0, 2=12.7, 3=5, Index=$Particles.PartitionParticles.Partition, Num=::NewEnumerator2)
          Global Phase=&Add_Float(A=$Particles.NormalizedAge, B=&Multiply_Float(A=$Emitter.Age, B=0.4))
          "Waveform [1]"=::NewEnumerator9
          "[1] Frequency"=3.5
          "[1] Amplitude Min/Max"=<-1 1>
          "[1] Scale"=20
         )
         Y=&Waveform(
          Global Input Position=&MergeHelper_FloatSwitch(0=1.2, 1=0, 2=12.7, 3=5, Index=$Particles.PartitionParticles.Partition, Num=::NewEnumerator2)
          Global Phase=&Add_Float(A=$Particles.NormalizedAge, B=&Multiply_Float(A=$Emitter.Age, B=0.4))
          "Waveform [1]"=::NewEnumerator10
          "[1] Frequency"=3
          "[1] Amplitude Min/Max"=<-1 1>
          "[1] Scale"=20
         )
         Z=&FloatFromCurve(
          FloatCurve=DataInterface() {propertyValues {Curve {keys [{arriveTangent=2.1415, interpMode=RCIM_Cubic, leaveTangent=2.1415, tangentMode=RCTM_User, time=0, value=0}, <at 1 1 RCIM_Cubic>]}}}
          Scale Curve=100
         )
        )
        B=&FN_FrontEndVector3Modifier(
         Gameplay=<0 0 0>
         Front End=&TransformVector(
          Vector=&MakeVector(
           X=0
           Y=&FloatFromCurve(
            FloatCurve=DataInterface() {propertyValues {Curve {keys [{arriveTangent=2.2182, time=0.1, value=0}, {leaveTangent=1.5818, time=1, value=1}]}}}
            Scale Curve=-50
           )
           Z=0
          )
          Destination Space=::Simulation
         )
        )
       )
       Alpha=&FloatFromCurve(FloatCurve=DataInterface() {propertyValues {Curve {keys [{arriveTangent=2.2182, time=0.1, value=0}, {leaveTangent=1.5818, time=1, value=1}]}}})
      )
      B=&RotateVectorByQuaternion(
       Vector To Rotate=&TransformVector(
        Vector=&Add_Vector(
         A=&Multiply_VectorByFloat(
          Vector=&Multiply_Vector(
           A=<-2 25 2>
           B=&MergeHelper_VectorSwitch(0=<1 1 6>, 1=<1 1 6>, 2=<1 -1 6>, 3=<1 -1 6>, Index=$Particles.PartitionParticles.Partition, Num=::NewEnumerator2)
          )
          Float=1
         )
         B=<0 0 0>
        )
        Destination Space=::Local
       )
       Quaternion=$Output.SamplePosFromOtherParticle001.ParticleQuat
      )
     )
    )) {isSetParameters=true}
   &DynamicMaterialParameters(
    Index 0 Param 1=&RandomRangeFloat(Minimum=0, Maximum=50000, Randomness Mode=::NewEnumerator1, Random Seed=$Particles.PartitionParticles.Partition, Fixed Random Seed=true)
    Module.Index0Float0Write=true
    Module.Index0Float1Write=true
    Module.Index0Float2Write=true
    Module.Index0Float3Write=true
   )
  ]
  rendererClasses [&NiagaraRibbonRendererProperties]
  renderers [{rendererClass=&NiagaraRibbonRendererProperties}]
 }
 Soul {
  emitterUpdateScript [
   &EmitterState()
   &SpawnBurst_Instantaneous(Spawn Count=2, Use Spawn Probability=false, UseLoopCountLimit=true)
   &SpawnBurst_Instantaneous(Spawn Count=4, Spawn Group=1, Use Spawn Probability=false, UseLoopCountLimit=true) {module=SpawnBurst_Instantaneous001}
   &InterpolateOverTime(
    Rate Of Change=10
    Target Vector Value:Vector3f=&Add_Vector(
     A=&FN_FrontEndVector3Modifier(
      Gameplay=&ReturnCameraVector(
       Camera Vectors=::NewEnumerator1
       Coordinate Space=::Simulation
       CameraQuery=DataInterface() {propertyValues {bRequireCurrentFrameData=false}}
      )
      Front End=&RotateVectorByQuaternion(Vector To Rotate=<0 1 0>, Quaternion=$Engine.Owner.Rotation)
     )
     B=$Engine.Owner.Velocity
    )
    Variable Type=::NewEnumerator2
   )
   /SetVariables(
    Emitter.AppearCurve:NiagaraFloat=&FloatFromCurve(
     FloatCurve=DataInterface() {propertyValues {Curve {keys [
         <at 0 25 RCIM_Cubic>
         {arriveTangent=-42, interpMode=RCIM_Cubic, leaveTangent=-42, time=0.2826, value=3}
         {interpMode=RCIM_Cubic, leaveTangent=-21.7231, time=0.5, value=4}
        ]}}}
     CurveIndex=&Subtract_Float(A=$Emitter.Age, B=$User.Delay)
    )
    Emitter.FireIntensity:NiagaraFloat=&Add_Float(
     A=&FloatFromCurve(
      FloatCurve=DataInterface() {propertyValues {Curve {keys [
          {arriveTangent=-87.1789, interpMode=RCIM_Cubic, time=0.1, value=-3}
          {arriveTangent=20, interpMode=RCIM_Cubic, leaveTangent=20, time=0.3279, value=20}
          {interpMode=RCIM_Cubic, leaveTangent=-92.9908, time=0.5, value=5}
         ]}}}
      CurveIndex=&Subtract_Float(A=$Emitter.Age, B=$User.Delay)
     )
     B=&Multiply_Float(A=$User.ParaGlideLeanAlpha, B=3)
    )
   ) {isSetParameters=true}
  ]
  particleSpawnScript [
   /SetVariables(Particles.VisibilityTag=&SpawnGroup()) {isSetParameters=true}
   &PartitionParticles(Number Of Partitions=4)
   &FN_SetSpawnGroup(Number of Spawn Groups=2)
   &InitializeParticle(
    Lifetime=2
    Mesh Scale Mode=::NewEnumerator3
    Mesh Uniform Scale=1.3
    Ribbon Width=10
    Ribbon Width Mode=::NewEnumerator1
    Sprite Size Mode=::NewEnumerator3
    Uniform Sprite Size=40
    UsePositionOffset=true
    Write Material Random=true
   )
   &CameraOffset(Camera Offset Amount=&MergeHelper_FloatSwitch(0=0, 1=2, 2=?, 3=?, Index=$Particles.VisibilityTag, Num=::NewEnumerator0))
   /SetVariables(Particles.RotAmount=1) {isSetParameters=true}
  ]
  particleUpdateScript [
   &ParticleState(Kill Particles When Lifetime Has Elapsed=false, Let Infinitely Lived Particles Die When Emitter Deactivates=true)
   &SkeletalMeshLocation(
    Bone / Socket Index=&SpawnGroup()
    Bone Sampling Mode=::NewEnumerator4
    MaskSpawnGroup=false
    Orientation Sampling=::NewEnumerator1
    Sampled Position Offset=<0 0 30>
    Sampled Position Offset Coordinate Space:ENiagaraSkelMeshTransforms=::NewEnumerator1
    Skeletal Mesh=$System.Skeletal Mesh
    Use Sampled Mesh Attribute as Ribbon ID=false
   )
   &ScaleMeshSize(Scale Factor=&VectorFromFloat(Value=&FloatFromCurve(
      FloatCurve=DataInterface() {propertyValues {Curve {keys [
          {arriveTangent=6.572, interpMode=RCIM_Cubic, tangentMode=RCTM_User, time=0, value=0}
          {interpMode=RCIM_Cubic, tangentMode=RCTM_User, time=0.3, value=1}
         ]}}}
      CurveIndex=$Emitter.Age
     )))
   /SetVariables(Particles.Wobble:Quat4f=&MultiplyQuaternion(
     Quaternion A=&FindQuatBetweenTwoVectors(
      From Vector=&MakeVector(
       X=&Sine(
        Normalized Angle=&Add_Float(
         A=$System.FN_Integrator.Value
         B=&RandomRangeFloat(Minimum=0, Maximum=1, Randomness Mode=::NewEnumerator1, Random Seed=6, Fixed Random Seed=true)
        )
        Period=&MakeCustomFloatFromBool(
         Bool=&RandomBool(
          Randomness Mode=::NewEnumerator1
          Random Seed=&Add_Integer(A=5, B=&SpawnGroup())
          Fixed Random Seed=true
         )
         True Float=1
         False Float=-1
        )
        Scale=1
        Bias=0
       )
       Y=&Cosine(
        Normalized Angle=&Add_Float(
         A=$System.FN_Integrator.Value
         B=&RandomRangeFloat(Minimum=0, Maximum=1, Randomness Mode=::NewEnumerator1, Random Seed=&SpawnGroup(), Fixed Random Seed=true)
        )
        Period=&MakeCustomFloatFromBool(Bool=&RandomBool(), True Float=1, False Float=-1)
        Scale=1
        Bias=0
       )
       Z=&RandomRangeFloat(
        Minimum=6
        Maximum=8
        Randomness Mode=::NewEnumerator1
        Random Seed=&Add_Integer(A=8, B=&SpawnGroup())
        Fixed Random Seed=true
       )
      )
      To Vector=<0 0 1>
     )
     Quaternion B=$Particles.Initial.MeshOrientation
    )) {isSetParameters=true}
   /BRCosmetics/Athena/Effects/Particles/Skydiving/NS_Contrail_RedPepper/UpdateMeshOrientation_RedPepper_TEMP(
    Bank Rate=0
    Banking Angle Min / Max:Vector2f=<0 0>
    "Foward Vector (Mesh Space)":Vector3f=&Lerp_Vector(A=<0.5 0 0>, B=<1 0 0>, Alpha=$System.InterpolateOverTime.InterpolateOverTime_Float.Moving Average)
    Orientation Method:ENiagara_UpdateMeshOrientationMode=::NewEnumerator3
    Particle Velocity:Vector3f=&Multiply_Vector(A=$Emitter.InterpolateOverTime.InterpolateOverTime_V3_.Moving Average, B=<1 1 0>)
    Rotation Decay Rate In Degrees:NiagaraFloat=&Lerp_Float(A=150, B=500, Alpha=$System.Dive)
    Upright Roll Correction Speed:NiagaraFloat=&Lerp_Float(A=25, B=90000, Alpha=$System.Dive)
   )
   &InterpolateOverTime(Rate Of Change=5, Target Quat Value:Quat4f=$Particles.MeshOrientation, Variable Type=::NewEnumerator5) {module=InterpolateOverTime002}
   /SetVariables(Particles.New Orient:Quat4f=$Particles.InterpolateOverTime002.InterpolateOverTime_Quaternion.Moving Average) {isSetParameters=true}
   /SetVariables(Particles.MeshOrientation=&MultiplyQuaternion(
     Quaternion A=$Particles.Wobble
     Quaternion B=/Niagara/DynamicInputs/TypeConversions/MakeCustomQuaternionFromBool(
      Bool:NiagaraBool=&Boolean_OrOperation(Bool1=$User.bIsFrontEndPreview, Bool2=$User.bIsFrontEnd)
      True Quat:Quat4f=&MultiplyQuaternion(
       Quaternion A=$Engine.Owner.Rotation
       Quaternion B=&MakeQuaternion(Quaternion From=3, Axis=<1 0 0>, Angle=0, From=<1 0 0>, To=<0 1 0>, Angle Type=::NewEnumerator0, Normalize Vectors=false)
      )
      False Quat:Quat4f=$Particles.New Orient
     )
    )) {isSetParameters=true}
   &ScaleColor(
    Scale RGB=&MakeVectorFromLinearColorRGB(Color=&LerpLinearColors(
      StartColor=<rgba 2 0.646 0.7569 1>
      EndColor=<rgba 4 0.84 0.84 1>
      LerpFactor=$System.InterpolateOverTime.InterpolateOverTime_Float.Moving Average
     ))
    ScaleA=true
    ScaleRGB=true
   )
   &DynamicMaterialParameters(
    Index 0 Param 2=0.5
    Index 0 Param 3=&RandomRangeFloat(Minimum=0.1, Maximum=1)
    Index 0 Param 4=&RandomRangeFloat(Minimum=0.75, Maximum=1)
    Module.Param0WriteEnabled=true
    Module.Param1WriteEnabled=true
    Module.Param2WriteEnabled=true
    Module.Param3WriteEnabled=true
   )
   /SetVariables(Particles.SplineFlameDyn_01:Vector4f=<1 1 1 0>) {isSetParameters=true}
   /SetVariables(
    Particles.GlowColor=&MakeLinearColorFromVectorAndFloat(
     "Vector (RGB)"=&MakeVectorFromLinearColorRGB(Color=<rgba 1 0.5781 0.7153 1>)
     "Float (Alpha)"=&FloatFromCurve(
      FloatCurve=DataInterface() {propertyValues {Curve {keys [<at 0.2 0 RCIM_Cubic>,<at 0.6 1 RCIM_Cubic>]}}}
      Scale Curve=&Multiply_Float(
       A=1
       B=&Waveform(
        Global Input Position=&RandomRangeFloat(Minimum=0, Maximum=100)
        "Waveform [1]"=::NewEnumerator12
        "[1] Frequency"=6
        "[1] Amplitude Min/Max"=<0.5 1>
       )
      )
     )
    )
    Particles.GlowPosition:NiagaraPosition=&AddVectorToPosition(
     Position=$Particles.Position
     Vector=&RotateVectorByQuaternion(
      Vector To Rotate=&Multiply_Vector(
       A=<2 7 -4>
       B=&MergeHelper_VectorSwitch(0=<1 1 1>, 1=<1 1 1>, 2=<1 -1 1>, 3=<1 -1 1>, Index=$Particles.PartitionParticles.Partition, Num=::NewEnumerator2)
      )
      Quaternion=$Particles.MeshOrientation
     )
    )
    Particles.GlowSize=&Multiply_Vector2D_ByFloat(
     Float=&FloatFromCurve(FloatCurve=DataInterface() {propertyValues {Curve {keys [<at 0.3 0 RCIM_Cubic>,<at 0.6 1 RCIM_Cubic>]}}})
     Vector2D=&Vector2DFromFloat(Value=9)
    )
   ) {isSetParameters=true}
   /SetVariables(Particles.SkullOverlayColor:LinearColor=&MakeLinearColorFromVectorAndFloat(
     "Vector (RGB)"=&MakeVectorFromLinearColorRGB(Color=<rgba 1 1 1 1>)
     "Float (Alpha)"=&FloatFromCurve(
      FloatCurve=DataInterface() {propertyValues {Curve {keys [
          {arriveTangent=4.9861, interpMode=RCIM_Cubic, time=0.2, value=0}
          {arriveTangent=2.7928, interpMode=RCIM_Cubic, leaveTangent=2.7928, tangentMode=RCTM_User, time=0.9, value=1}
         ]}}}
      CurveIndex=&Subtract_Float(A=$Emitter.Age, B=$User.Delay)
     )
    )) {isSetParameters=true}
  ]
  rendererClasses [&NiagaraMeshRendererProperties, &NiagaraSpriteRendererProperties]
  renderers [
   {rendererClass=&NiagaraMeshRendererProperties, rendererIndex=0}
   {rendererClass=&NiagaraMeshRendererProperties, rendererIndex=1}
   {rendererClass=&NiagaraMeshRendererProperties, rendererIndex=2}
   {rendererClass=&NiagaraMeshRendererProperties, rendererIndex=3}
   {rendererClass=&NiagaraMeshRendererProperties, rendererIndex=4}
   {rendererClass=&NiagaraSpriteRendererProperties, rendererIndex=5}
  ]
 }
}
summary {
 systemName=NS_HomeRange_Contrail
 emitters [
  {emitterName=Soul, rendererClasses [&NiagaraMeshRendererProperties, &NiagaraSpriteRendererProperties]}
  {emitterName=Fire, rendererClasses [&NiagaraSpriteRendererProperties]}
  {emitterName=Feathers, rendererClasses [&NiagaraSpriteRendererProperties]}
  {emitterName=FireTrail_And_Dots, rendererClasses [&NiagaraMeshRendererProperties]}
  {emitterName=HornTrails, rendererClasses [&NiagaraRibbonRendererProperties]}
  {
   emitterName=Glow
   rendererClasses [
    &NiagaraLightRendererProperties
    &NiagaraSpriteRendererProperties
    &NiagaraMeshRendererProperties
   ]
  }
  {emitterName=Anticipation, rendererClasses [&NiagaraSpriteRendererProperties]}
  {emitterName=Embers, rendererClasses [&NiagaraSpriteRendererProperties]}
 ]
 userVariables [
  User.bIsFrontEnd {type=NiagaraBool}
  User.bIsFrontEndPreview {type=NiagaraBool}
  User.Delay {type=NiagaraFloat, defaultValue=0.6, description=""}
  User.ParaGlideLeanAlpha {type=NiagaraFloat, defaultValue=0, description=""}
 ]
}
